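# Editor Utility Widget for the IK/FK matching tool that shows one labeled row with "To Ik" and "To Fk" check boxes.
# A state change on either box sets the other to Unchecked, so the two act as exclusive options, though both can be off.
# On Construct, the CheckBoxText variable plus a " : " suffix becomes both the row label and its tooltip.
# The widget itself contains no matching logic.
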EditorUtilityWidgetBlueprint /AnimInEdTools/IkFkMatchingTool/EUW_TolkFkCheckBox
graphs [EventGraph {nodes [
   "On Check State Changed (ToFk_CheckBox)"!() -> (OutputDelegate, then, bIsChecked:bool) {is=title K2Node_ComponentBoundEvent}
   > Set CheckedState!(
    execute
    CheckedState:ECheckBoxState Enum=Unchecked
    self:Check Box Object Reference=Get ToIk_CheckBox() -> (ToIk_CheckBox:Editor Utility Check Box Object Reference)
   ) -> (then, Output_Get:ECheckBoxState Enum=Unchecked)
   "On Check State Changed (ToIk_CheckBox)"!() -> (OutputDelegate, then, bIsChecked:bool) {is=title K2Node_ComponentBoundEvent}
   > Set CheckedState(CheckedState=Unchecked, self=Get ToFk_CheckBox() -> (ToFk_CheckBox:Editor Utility Check Box Object Reference)) -> (Output_Get=Unchecked)
   Event Construct()
   n1 = Get CheckBoxText() -> (CheckBoxText:Text)
   n2 = "To Text (String)"(InString=Append(A="To String (Text)"(InText=@n1), B=" : "))
   n3 = Get CheckBox_TextBlock() -> (CheckBox_TextBlock:Text Object Reference)
   > Set Text(Text=@n2, self=@n3)
   > Set ToolTipText!(execute, ToolTipText:Text=@n2, self:Widget Object Reference=@n3) -> (then, Output_Get:Text)
   Event PreConstruct()
   Event Tick()
  ]}]
variables [CheckBoxText {type=string, default=""}]
widgets {
 info {rootWidgetClass=&VerticalBox, widgetCount=7}
 widgets [
  VerticalBox_23 {widgetClassPath=&VerticalBox, widget=VerticalBox_23}
  HorizontalBox_68 {
   widgetClassPath=&HorizontalBox
   parent=VerticalBox_23
   widget=HorizontalBox_68
   slot {class=VerticalBoxSlot, properties {horizontalAlignment=HAlign_Center, padding=<ltrb 3 3 3 3>}}
  }
  CheckBox_TextBlock {
   widgetClassPath=&TextBlock
   bIsVariable=true
   parent=HorizontalBox_68
   widget=CheckBox_TextBlock
   properties {
    text=CheckboxText
    font {
     bForceMonospaced=false
     bMaterialIsStencil=false
     fontMaterial=null
     fontObject=&Roboto
     letterSpacing=0
     monospacedWidth=1
     size=13.5
     skewAmount=0
     typefaceFontName=Regular
     outlineSettings {
      bApplyOutlineToDropShadows=false
      bMiteredCorners=false
      bSeparateFillAlpha=false
      outlineMaterial=null
      outlineSize=0
      outlineColor=<rgba 0 0 0 1>
     }
    }
   }
   slot {class=HorizontalBoxSlot, properties {padding=<ltrb 0 0 10 0>}}
  }
  ToIk_CheckBox {
   widgetClassPath=&EditorUtilityCheckBox
   bIsVariable=true
   parent=HorizontalBox_68
   widget=ToIk_CheckBox
   slot {class=HorizontalBoxSlot, properties {horizontalAlignment=HAlign_Center, verticalAlignment=VAlign_Center, padding=<ltrb 0 0 5 0>}}
  }
  ToIK_TextBlock {
   widgetClassPath=&TextBlock
   parent=ToIk_CheckBox
   widget=ToIK_TextBlock
   properties {
    text=To Ik
    font {
     bForceMonospaced=false
     bMaterialIsStencil=false
     fontMaterial=null
     fontObject=&Roboto
     letterSpacing=0
     monospacedWidth=1
     size=13.5
     skewAmount=0
     typefaceFontName=Regular
     outlineSettings {
      bApplyOutlineToDropShadows=false
      bMiteredCorners=false
      bSeparateFillAlpha=false
      outlineMaterial=null
      outlineSize=0
      outlineColor=<rgba 0 0 0 1>
     }
    }
   }
   slot {class=PanelSlot}
  }
  ToFk_CheckBox {
   widgetClassPath=&EditorUtilityCheckBox
   bIsVariable=true
   parent=HorizontalBox_68
   widget=ToFk_CheckBox
   slot {class=HorizontalBoxSlot, properties {horizontalAlignment=HAlign_Center, verticalAlignment=VAlign_Center}}
  }
  ToFK_TextBlock {
   widgetClassPath=&TextBlock
   parent=ToFk_CheckBox
   widget=ToFK_TextBlock
   properties {
    text=To Fk
    font {
     bForceMonospaced=false
     bMaterialIsStencil=false
     fontMaterial=null
     fontObject=&Roboto
     letterSpacing=0
     monospacedWidth=1
     size=13.5
     skewAmount=0
     typefaceFontName=Regular
     outlineSettings {
      bApplyOutlineToDropShadows=false
      bMiteredCorners=false
      bSeparateFillAlpha=false
      outlineMaterial=null
      outlineSize=0
      outlineColor=<rgba 0 0 0 1>
     }
    }
   }
   slot {class=PanelSlot}
  }
 ]
}
