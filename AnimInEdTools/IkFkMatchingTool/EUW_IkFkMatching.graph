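# Editor utility widget that matches and bakes Control Rig limbs between IK and FK in Sequencer.
# Limb checkboxes come from the Animatable Creature's matching data asset, with Animation Browser as fallback; modular rigs are supported.
# Match resolves the frame range, clears target keys as needed, then steps time with tick delays, matching at intervals or existing keys.
# It finally keys the IK/FK switches; embedded Python fills custom ranges from selected keys.
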EditorUtilityWidgetBlueprint /AnimInEdTools/IkFkMatchingTool/EUW_IkFkMatching
graphs [
 Add Module Path To Control Names {nodes [
   n1 = Add Module Path To Control Names!() -> (then, Module Path:String, Controls:Array of Names, Only Module Name:bool=~) {is=title K2Node_FunctionEntry}
   > Clear(TargetArray:Array of Names=Get Control Names() -> (Control Names:Array of Names))
   > n2 = Set Module Path Shorten(Module Path Shorten:String=@n1) -> (Output_Get:String)
   > n3 = Branch(Condition=@n1.Only Module Name)
   > n4 = Set Module Path Shorten(Module Path Shorten=Split(SourceString=@n2, InStr=/, SearchDir=FromEnd).RightS)
   n5 = For Each Loop(Exec=[@n4.then, @n3.else], Array:Array of Names=@n1.Controls) -> (Array Element:Name)
   Return Node!(execute=@n5.Completed, Output Controls:Array of Names=Get Control Names())
   n6 = "To String (Name)"(InName=@n5)
   > n7 = Branch(Condition=Contains(SearchIn=@n6, Substring=/))
   > Add(TargetArray:Array of Names=Get Control Names(), NewItem:"Name (by ref)"=@n5) {is=title K2Node_CallArrayFunction}
   n8 = Get Module Path Shorten() -> (Module Path Shorten:String)
   Add(execute=@n7.else, TargetArray=Get Control Names(), NewItem=String To Name(InString=Append!(A=@n8, B=/, C:String=@n6) -> (ReturnValue)))
  ]}
 EventGraph {nodes [
   "On Clicked (Cancel_Button)"() {is=title K2Node_ComponentBoundEvent}
   > RemoveFromParent()
   n1 = "On Check State Changed (TimeSliderCheckBox)"!() -> (OutputDelegate, then, bIsChecked:bool) {is=title K2Node_ComponentBoundEvent}
   > Branch(Condition=@n1.bIsChecked)
   > SetCheckedState!(
    execute
    self:Check Box Object Reference=Get CurrentFrameCheckBox() -> (CurrentFrameCheckBox:Editor Utility Check Box Object Reference)
    InCheckedState:ECheckBoxState Enum=Unchecked
   ) -> (then)
   > SetCheckedState(self=Get SpecifiedRangeCheckBox() -> (SpecifiedRangeCheckBox:Editor Utility Check Box Object Reference), InCheckedState=Unchecked)
   > Set Visibility(Visibility=Collapsed, self=Get FrameRateHorizontalBox() -> (FrameRateHorizontalBox:Horizontal Box Object Reference))
   n2 = "On Check State Changed (CurrentKeysOnlyCheckBox)"!() -> (OutputDelegate, then, bIsChecked:bool) {is=title K2Node_ComponentBoundEvent}
   > n3 = Branch(Condition=@n2.bIsChecked)
   > SetCheckedState(self=Get CurrentFrameCheckBox(), InCheckedState=Unchecked)
   > Set IsReadOnly!(
    execute
    IsReadOnly:bool=true
    self:Text Box Object Reference=Get KeyEveryTextBox() -> (KeyEveryTextBox:Editor Utility Editable Text Box Object Reference)
   ) -> (then, Output_Get:bool)
   Set IsReadOnly(execute=@n3.else, IsReadOnly=false, self=Get KeyEveryTextBox())
   n4 = "On Check State Changed (CurrentFrameCheckBox)"!() -> (OutputDelegate, then, bIsChecked:bool) {is=title K2Node_ComponentBoundEvent}
   > Branch(Condition=@n4.bIsChecked)
   > SetCheckedState(self=Get CurrentKeysOnlyCheckBox() -> (CurrentKeysOnlyCheckBox:Editor Utility Check Box Object Reference), InCheckedState=Unchecked)
   > SetCheckedState(self=Get TimeSliderCheckBox() -> (TimeSliderCheckBox:Editor Utility Check Box Object Reference), InCheckedState=Unchecked)
   > SetCheckedState(self=Get SpecifiedRangeCheckBox(), InCheckedState=Unchecked)
   > Set Visibility(Visibility=Collapsed, self=Get FrameRateHorizontalBox())
   "On Clicked (ACBrowse_Button)"() {is=title K2Node_ComponentBoundEvent}
   n5 = Get Animatable Creature() -> (Animatable Creature:Animatable Creature Object Reference)
   > SyncBrowserToObjects(AssetPaths=Make Array("[0]":String=Get Object Path String(Object=@n5)) -> (Array:Array of Strings))
   "On Clicked (Match_Button)"() {is=title K2Node_ComponentBoundEvent}
   > Initialization()
   > Define Matching Data()
   > n6 = Branch(Condition=Is Not Empty!(TargetArray:Array of Ik Fk Matching Data Structures=Get Ik Fk Matching Data() -> (Ik Fk Matching Data:Array of Ik Fk Matching Data Structures)) -> (ReturnValue:bool))
   > Delay!(execute, WorldContextObject:Object Reference, Duration=0.1) -> (then)
   > Run Match()
   Set Output Log Message!(execute=@n6.else, self, Output Message:String=Get Output Message() -> (Output Message:String)) -> (then) {is=title K2Node_CallFunction}
   n7 = "On Check State Changed (SpecifiedRangeCheckBox)"!() -> (OutputDelegate, then, bIsChecked:bool) {is=title K2Node_ComponentBoundEvent}
   > n8 = Branch(Condition=@n7.bIsChecked)
   n9 = Get FrameRateHorizontalBox()
   > Set Visibility(Visibility=Visible, self=@n9)
   > SetCheckedState(self=Get TimeSliderCheckBox(), InCheckedState=Unchecked)
   > n10 = SetCheckedState(self=Get CurrentFrameCheckBox(), InCheckedState=Unchecked)
   n11 = Execute Python Script!(
    execute=[@n10.then, @n12.then]
    PythonScript="import unreal\r\n\r\n## 1. Get the current level sequence\r\nlevel_sequence = unreal.LevelSequenceEditorBlueprintLibrary.get_current_level_sequence()\r\n\r\n## 2. Initialize variables\r\nselected_frames = []\r\nmin_frame = 0\r\nmax_frame = 0\r\n\r\n## Set the fallback values to the sequence's playback range\r\nmin_frame = level_sequence.get_playback_start()\r\nmax_frame = level_sequence.get_playback_end()\r\n\r\n## Get all channels that have selected keys\r\nkeyed_channels = unreal.LevelSequenceEditorBlueprintLibrary.get_channels_with_selected_keys()\r\n\r\nif keyed_channels:\r\n    for channel_proxy in keyed_channels:\r\n        ## Get indices for selected keys in this channel\r\n        selected_indices = unreal.LevelSequenceEditorBlueprintLibrary.get_selected_keys(channel_proxy)\r\n        \r\n        ## Get the actual channel data\r\n        channel_obj = channel_proxy.section.get_channel(channel_proxy.channel_name)\r\n        \r\n        ## Convert indices to key objects and store their frame values\r\n        key_objects = channel_obj.get_keys_by_index(selected_indices)\r\n        for key in key_objects:\r\n            selected_frames.append(key.get_time().frame_number.value)\r\n\r\n## 3. If keys were selected, overwrite the fallback values\r\nif selected_frames:\r\n    min_frame = min(selected_frames)\r\n    max_frame = max(selected_frames)\r\n\r\nif min_frame == max_frame:\r\n    min_frame = level_sequence.get_playback_start()\r\n    max_frame = level_sequence.get_playback_end()"
   ) -> (then, ReturnValue, min_frame:int, max_frame:int) {outputs [min_frame,max_frame]}
   > Set Text(
    Text="To Text (String)"(InString="To String (Integer)"(InInt=@n11.min_frame))
    self:Text Box Object Reference=Get StartFrameTextBox() -> (StartFrameTextBox:Editor Utility Editable Text Box Object Reference)
   )
   > Set Text(
    Text="To Text (String)"(InString="To String (Integer)"(InInt=@n11.max_frame))
    self=Get EndFrameTextBox() -> (EndFrameTextBox:Editor Utility Editable Text Box Object Reference)
   )
   Set Visibility(execute=@n8.else, Visibility=Collapsed, self=@n9)
   n12 = "On Clicked (SetFrameRangeFromSelected_Button)"() {is=title K2Node_ComponentBoundEvent}
   Delete Existing Keys() {is=title K2Node_CustomEvent}
   > n13 = For Each Loop(Array:Array of Ik Fk Matching Data Structures=Get Ik Fk Matching Data()) -> (Array Element:Ik Fk Matching Data Structure)
   > n14 = Sequence()
   n15 = Break Ik Fk Matching Data!(IkFkMatchingData:"Ik Fk Matching Data Structure (by ref)"=@n13) -> (
    Name_19_#1:Name
    IkFkSwitchAttr_20_#2:Name
    MatchIKEvent_21_#3:Name
    MatchFKEvent_22_#4:Name
    IKControls_10_#5:Array of Names
    FKControls_12_#6:Array of Names
   ) {is=title K2Node_BreakStruct}
   > Set Ik Controls(Ik Controls:Array of Names=@n15.IKControls_10_#5) -> (Output_Get:Array of Names)
   > Set Fk Controls(Fk Controls:Array of Names=@n15.FKControls_12_#6) -> (Output_Get:Array of Names)
   n16 = Get IK FK Switch Channels() -> (IK FK Switch Channels:Map of Movie Scene Scripting Bool Channel Object References to Booleans)
   > n17 = Keys!(execute, TargetMap:Map of Movie Scene Scripting Bool Channel Object References to Booleans=@n16) -> (then, Keys:Array of Movie Scene Scripting Bool Channel Object References)
   > n18 = Branch(Condition=Find(
     TargetMap:Map of Movie Scene Scripting Bool Channel Object References to Booleans=@n16
     Key:"Movie Scene Scripting Bool Channel Object Reference (by ref)"="Get (a copy)"(Array:Array of Movie Scene Scripting Bool Channel Object References=@n17, Dimension 1=@n13.Array Index) -> (Output:Movie Scene Scripting Bool Channel Object Reference)
    ) -> (Value:bool))
   > n19 = Branch(Condition=Get IsModular() -> (IsModular:bool))
   n20 = "Get (a ref)"(Array:Array of Strings=Get Module Paths() -> (Module Paths:Array of Strings), Dimension 1=@n13.Array Index) -> (Output:"String (by ref)")
   n21 = Get Ik Controls() -> (Ik Controls:Array of Names)
   > n22 = Add Module Path To Control Names!(execute, self, Module Path:String=@n20, Controls:Array of Names=@n21, Only Module Name:bool=true) -> (then, Output Controls:Array of Names) {is=title K2Node_CallFunction}
   > Set Controls To Key(Controls To Key:Array of Names=@n22) -> (Output_Get:Array of Names)
   Set Controls To Key(execute=@n19.else, Controls To Key=@n21)
   n23 = Branch(execute=@n18.else, Condition=Get IsModular())
   n24 = Get Fk Controls() -> (Fk Controls:Array of Names)
   > n25 = Add Module Path To Control Names(Module Path=@n20, Controls=@n24, Only Module Name=true)
   > Set Controls To Key(Controls To Key=@n25)
   Set Controls To Key(execute=@n23.else, Controls To Key=@n24)
   n26 = GetSections!(
    execute=@n14.then_1
    Track:Movie Scene Track Object Reference=Get Track() -> (Track:Movie Scene Control Rig Parameter Track Object Reference)
   ) -> (then, ReturnValue:Array of Movie Scene Section Object References)
   > n27 = For Each Loop(Array:Array of Movie Scene Section Object References=@n26) -> (Array Element:Movie Scene Section Object Reference)
   > n28 = GetAllChannels!(execute, Section:Movie Scene Section Object Reference=@n27) -> (then, ReturnValue:Array of Movie Scene Scripting Channel Object References)
   > n29 = For Each Loop(Array:Array of Movie Scene Scripting Channel Object References=@n28) -> (Array Element:Movie Scene Scripting Channel Object Reference)
   > n30 = For Each Loop(Array:Array of Names=Get Controls To Key() -> (Controls To Key:Array of Names)) -> (Array Element:Name)
   > Clear(TargetArray:Array of Movie Scene Scripting Key Object References=Get Keys To Delete() -> (Keys To Delete:Array of Movie Scene Scripting Key Object References))
   > Branch(Condition=Contains(SearchIn=GetDisplayName(Object=@n29), Substring="To String (Name)"(InName=@n30)))
   > n31 = Cast To MovieSceneScriptingFloatChannel!(execute, Object=@n29) -> (then, CastFailed, AsMovie Scene Scripting Float Channel:Movie Scene Scripting Float Channel Object Reference) {is=title K2Node_DynamicCast}
   > n32 = For Each Loop(Array:Array of Movie Scene Scripting Key Object References="Get Keys (Float)"!(self:Movie Scene Scripting Float Channel Object Reference=@n31) -> (ReturnValue:Array of Movie Scene Scripting Key Object References) {is=title K2Node_CallFunction}) -> (Array Element:Movie Scene Scripting Key Object Reference)
   n33 = For Each Loop(Exec=@n32.Completed, Array=Get Keys To Delete())
   > "Remove Key (Float)"!(execute, self:Movie Scene Scripting Float Channel Object Reference=@n31, Key:Movie Scene Scripting Key Object Reference=@n33) -> (then)
   n34 = Cast To MovieSceneScriptingFloatKey!(execute=@n32.LoopBody, Object=@n32) -> (then, CastFailed, AsMovie Scene Scripting Float Key:Movie Scene Scripting Float Key Object Reference) {is=title K2Node_DynamicCast}
   n35 = "Get Time (Float)"!(self:Movie Scene Scripting Float Key Object Reference=@n34, TimeUnit:EMovieSceneTimeUnit Enum=DisplayRate) -> (ReturnValue_FrameNumber_Value:int, ReturnValue_SubFrame:float=~) {is=title K2Node_CallFunction}
   n36 = Get!() -> (Playback Range_bHasStart:bool=~, Playback Range_bHasEnd:bool=~, Playback Range_InclusiveStart:int, Playback Range_ExclusiveEnd:int)
   > Branch(Condition=AND Boolean(A="integer >= integer"(A=@n35, B=@n36.Playback Range_InclusiveStart), B="integer <= integer"(A=@n35, B=@n36.Playback Range_ExclusiveEnd)))
   > Add(
    TargetArray:Array of Movie Scene Scripting Key Object References=Get Keys To Delete()
    NewItem:"Movie Scene Scripting Key Object Reference (by ref)"=@n32
   )
   n37 = Cast To MovieSceneScriptingBoolChannel!(execute=@n31.CastFailed, Object=@n29) -> (then, CastFailed, AsMovie Scene Scripting Bool Channel:Movie Scene Scripting Bool Channel Object Reference) {is=title K2Node_DynamicCast}
   > n38 = For Each Loop(Array="Get Keys (Bool)"!(self:Movie Scene Scripting Bool Channel Object Reference=@n37) -> (ReturnValue:Array of Movie Scene Scripting Key Object References) {is=title K2Node_CallFunction})
   n39 = For Each Loop(Exec=@n38.Completed, Array=Get Keys To Delete())
   > "Remove Key (Bool)"!(execute, self:Movie Scene Scripting Bool Channel Object Reference=@n37, Key:Movie Scene Scripting Key Object Reference=@n39) -> (then)
   n40 = Cast To MovieSceneScriptingBoolKey!(execute=@n38.LoopBody, Object=@n38) -> (then, CastFailed, AsMovie Scene Scripting Bool Key:Movie Scene Scripting Bool Key Object Reference) {is=title K2Node_DynamicCast}
   n41 = "Get Time (Bool)"!(self:Movie Scene Scripting Bool Key Object Reference=@n40, TimeUnit:EMovieSceneTimeUnit Enum=DisplayRate) -> (ReturnValue_FrameNumber_Value:int, ReturnValue_SubFrame:float=~) {is=title K2Node_CallFunction}
   n42 = Get() -> (Playback Range_bHasStart=~, Playback Range_bHasEnd=~)
   > Branch(Condition=AND Boolean(A="integer >= integer"(A=@n41, B=@n42.Playback Range_InclusiveStart), B="integer <= integer"(A=@n41, B=@n42.Playback Range_ExclusiveEnd)))
   > Add(TargetArray=Get Keys To Delete(), NewItem=@n38)
   n43 = Cast To MovieSceneScriptingIntegerChannel!(execute=@n37.CastFailed, Object=@n29) -> (then, CastFailed, AsMovie Scene Scripting Integer Channel:Movie Scene Scripting Integer Channel Object Reference) {is=title K2Node_DynamicCast}
   > n44 = For Each Loop(Array="Get Keys (Integer)"!(self:Movie Scene Scripting Integer Channel Object Reference=@n43) -> (ReturnValue:Array of Movie Scene Scripting Key Object References) {is=title K2Node_CallFunction})
   n45 = For Each Loop(Exec=@n44.Completed, Array=Get Keys To Delete())
   > "Remove Key (Integer)"!(execute, self:Movie Scene Scripting Integer Channel Object Reference=@n43, Key:Movie Scene Scripting Key Object Reference=@n45) -> (then)
   n46 = Cast To MovieSceneScriptingIntegerKey!(execute=@n44.LoopBody, Object=@n44) -> (then, CastFailed, AsMovie Scene Scripting Integer Key:Movie Scene Scripting Integer Key Object Reference) {is=title K2Node_DynamicCast}
   n47 = "Get Time (Integer)"!(self:Movie Scene Scripting Integer Key Object Reference=@n46, TimeUnit:EMovieSceneTimeUnit Enum=DisplayRate) -> (ReturnValue_FrameNumber_Value:int, ReturnValue_SubFrame:float=~) {is=title K2Node_CallFunction}
   n48 = Get() -> (Playback Range_bHasStart=~, Playback Range_bHasEnd=~)
   > Branch(Condition=AND Boolean(A="integer >= integer"(A=@n47, B=@n48.Playback Range_InclusiveStart), B="integer <= integer"(A=@n47, B=@n48.Playback Range_ExclusiveEnd)))
   > Add(TargetArray=Get Keys To Delete(), NewItem=@n44)
   Define Matching Data() {is=title K2Node_CustomEvent}
   > n49 = GetFocusedLevelSequence!(execute) -> (then, ReturnValue:Level Sequence Object Reference)
   > n50 = MakeRange!(
    execute
    Sequence:Movie Scene Sequence Object Reference=@n49
    StartFrame:int=Get() -> (Playback Range_bHasStart=~, Playback Range_bHasEnd=~).Playback Range_InclusiveStart
    Duration:int=1
   ) -> (then, ReturnValue:Sequencer Scripting Range Structure)
   > Set Range(Range:Sequencer Scripting Range Structure=@n50) -> (Output_Get:Sequencer Scripting Range Structure)
   > Clear(TargetArray:Array of Strings=Get Module Paths())
   > Clear!(execute, TargetMap:Map of Movie Scene Scripting Bool Channel Object References to Booleans=Get IK FK Switch Channels()) -> (then) {is=title K2Node_CallFunction}
   > Clear(TargetArray:Array of Ik Fk Matching Data Structures=Get Ik Fk Matching Data()) {is=title K2Node_CallArrayFunction}
   > n51 = For Each Loop(Array:Array of EUW Tolk Fk Check Box Object References=Get Limb Check Boxes() -> (Limb Check Boxes:Array of EUW Tolk Fk Check Box Object References)) -> (Array Element:EUW Tolk Fk Check Box Object Reference)
   Branch(execute=@n51.Completed, Condition=Is Not Empty(TargetArray=Get Ik Fk Matching Data()))
   > n52 = Branch(Condition=IsChecked(self=Get CurrentKeysOnlyCheckBox()))
   > Delete Existing Keys() {is=title K2Node_CallFunction}
   > SetText(
    self=Get Output_TextBlock() -> (Output_TextBlock:Rich Text Block Object Reference)
    InText="LOCTEXT(\"<Warning>Finding Existing Keys...</>\")"
   )
   > Delay(Duration=0.05)
   > Find Existing Keys!(execute, self) -> (then, Existing Keys Matching Data:Map of Names to IKFKMatching Frames Structures)
   n53 = "To String (Text)"(InText=Get Text(self=Get KeyEveryTextBox()))
   Branch(execute=@n52.else, Condition="Not Equal (Integer)"(A=String To Integer(InString=@n53), B=1))
   > Delete Existing Keys()
   n54 = IsChecked(self=Get ToIk_CheckBox!(self:EUW Tolk Fk Check Box Object Reference=@n51) -> (ToIk_CheckBox:Editor Utility Check Box Object Reference))
   Set Is IK(execute=@n51.LoopBody, Is IK:bool=@n54) -> (Output_Get:bool)
   n55 = Get ToFk_CheckBox!(self:EUW Tolk Fk Check Box Object Reference=@n51) -> (ToFk_CheckBox:Editor Utility Check Box Object Reference)
   > Branch(Condition=OR Boolean(A=@n54, B=IsChecked(self=@n55)))
   > n56 = GetSections(Track=Get Track())
   > n57 = For Each Loop(Array:Array of Movie Scene Section Object References=@n56) -> (Array Element:Movie Scene Section Object Reference)
   > n58 = GetAllChannels(Section=@n57)
   > n59 = For Each Loop(Array:Array of Movie Scene Scripting Channel Object References=@n58) -> (Array Element:Movie Scene Scripting Channel Object Reference)
   > Sequence()
   n60 = "Get (a copy)"(
    Array:Array of Ik Fk Matching Data Structures=Get Ik Fk Matching From Data Asset() -> (Ik Fk Matching From Data Asset:Array of Ik Fk Matching Data Structures)
    Dimension 1=@n51.Array Index
   ) -> (Output:Ik Fk Matching Data Structure)
   n61 = Break Ik Fk Matching Data(IkFkMatchingData=@n60)
   n62 = "To String (Name)"(InName=@n61.IkFkSwitchAttr_20_#2)
   n63 = Append(A="To String (Name)"(InName=@n61), B=/, C=@n62)
   n64 = Get IsModular()
   > Branch(Condition=Contains(SearchIn=GetDisplayName(Object=@n59), Substring=Select(Option 0:String=@n62, Option 1:String=@n63, Index=@n64) -> (ReturnValue:String)))
   > n65 = Cast To MovieSceneScriptingBoolChannel(Object=@n59)
   n66 = "Evaluate Keys (Bool)"!(
    self:Movie Scene Scripting Bool Channel Object Reference=@n65
    Range:Sequencer Scripting Range Structure=Get Range() -> (Range:Sequencer Scripting Range Structure)
    FrameRate:Frame Rate Structure=Get Frame Rate() -> (Frame Rate:Frame Rate Structure)
   ) -> (ReturnValue:Array of Booleans)
   > n67 = Branch(Condition="Not Equal (Boolean)"(A=Get Is IK() -> (Is IK:bool), B="Get (a ref)"(Array:Array of Booleans=@n66, Dimension 1=0) -> (Output:"Boolean (by ref)")) -> (ReturnValue=~))
   > Add!(
    execute
    TargetMap:Map of Movie Scene Scripting Bool Channel Object References to Booleans=Get IK FK Switch Channels()
    Key:"Movie Scene Scripting Bool Channel Object Reference (by ref)"=@n65
    Value:"Boolean (by ref)"=Get Is IK()
   ) -> (then) {is=title K2Node_CallFunction}
   > Add(TargetArray:Array of Ik Fk Matching Data Structures=Get Ik Fk Matching Data(), NewItem:"Ik Fk Matching Data Structure (by ref)"=@n60) {is=title K2Node_CallArrayFunction}
   n68 = Get Is IK()
   n69 = "To String (Name)"(InName=@n61)
   > Set Output Message(Output Message:String=Append!(
     A=Get Output Message()
     B="<Success>"
     C=@n69
     D:String=Select(Option 0=" successfully matched to FK.", Option 1=" successfully matched to IK.", Index=@n68)
     E:String="</>\r\n"
    ) -> (ReturnValue)) -> (Output_Get:String)
   > Branch(Condition=Get IsModular())
   > n70 = For Each Loop(Array:Array of Names=GetModuleNames!(self:Modular Rig Object Reference=Get Modular Rig() -> (Modular Rig:Modular Rig Object Reference)) -> (ReturnValue:Array of Names)) -> (Array Element:Name)
   n71 = "To String (Name)"(InName=@n70)
   > Branch(Condition="Equal Exactly (String)"(A=@n71, B=@n69))
   > Add(TargetArray:Array of Strings=Get Module Paths(), NewItem:"String (by ref)"=@n71)
   n72 = Get Is IK()
   Set Output Message(
    execute=@n67.else
    Output Message=Append(
     A=Get Output Message()
     B="<Fail>"
     C=@n69
     D=Select(Option 0=" is already in FK mode.", Option 1=" is already in IK mode.", Index=@n72)
     E="</>\r\n"
    )
   )
   Run Match() {is=title K2Node_CustomEvent}
   > SetText(self=Get Output_TextBlock(), InText="LOCTEXT(\"<Warning>Matching in progress...</>\")")
   > n73 = Branch(Condition=Is Not Empty(TargetArray=Get Ik Fk Matching Data()))
   > BeginTransaction(Context=EUW_IkFkMatching, Description="LOCTEXT(\"Matching IK FK Bake\")")
   > n74 = Branch(Condition=IsChecked(self=Get CurrentKeysOnlyCheckBox()))
   > Run Matching On Existing Keys()
   Set Current Local Time!(
    execute=[@n74.else, @n83.else]
    PlaybackParams_Frame_FrameNumber_Value:int=Get Current Matching Frame() -> (Current Matching Frame:int=0)
    PlaybackParams_Frame_SubFrame:float=0
    PlaybackParams_Time:float=0
    PlaybackParams_MarkedFrame:String
    PlaybackParams_Timecode:Timecode Structure=(Hours=0,Minutes=0,Seconds=0,Frames=0,Subframe=0,bDropFrameFormat=False)
    PlaybackParams_PositionType:EMovieScenePositionType Enum=Frame
    PlaybackParams_UpdateMethod:EUpdatePositionMethod Enum=Play
    PlaybackParams_bHasJumped:bool=False
    TimeUnit:EMovieSceneTimeUnit Enum=DisplayRate
   ) -> (then) {is=title K2Node_CallFunction}
   > DelayUntilNextTick!(execute, WorldContextObject:Object Reference) -> (then)
   n75 = Get KeyEveryTextBox()
   n76 = String To Integer(InString="To String (Text)"(InText=Get Text(self=@n75)))
   n77 = Get Current Matching Frame() -> (Current Matching Frame=0)
   > Set Current Matching Frame(Current Matching Frame:int=int + int(A=@n77, B=@n76)) -> (Output_Get:int=0)
   > n78 = Sequence!(execute) -> (then_0, then_1, then_2:Exec)
   n79 = Get() -> (Playback Range_bHasStart=~, Playback Range_bHasEnd=~)
   n80 = int + int(A=@n79.Playback Range_ExclusiveEnd, B=@n76)
   n81 = Get Current Matching Frame() -> (Current Matching Frame=0)
   > Branch(Condition=AND Boolean(A="integer > integer"(A=@n81, B=@n79.Playback Range_ExclusiveEnd), B="integer < integer"(A=@n81, B=@n80)))
   > Set Current Matching Frame(Current Matching Frame=@n79.Playback Range_ExclusiveEnd) -> (Output_Get=0)
   Run Matching(execute=@n78.then_1)
   n82 = Get() -> (Playback Range_bHasStart=~, Playback Range_bHasEnd=~)
   n83 = Branch(execute=@n78.then_2, Condition="integer >= integer"(A=@n77, B=int + int(A=@n82.Playback Range_ExclusiveEnd)))
   > Switch IK FK and Set Key()
   > Call IKFK Match Finished E Dispatcher!(execute:Exec, self:EUW Ik Fk Matching Object Reference) -> (then:Exec) {is=title K2Node_CallDelegate}
   > EndTransaction()
   Set Output Log Message(execute=@n73.else, Output Message=Get Output Message())
   Initialization() {is=title K2Node_CustomEvent}
   > BeginTransaction(Context=IKFkMatching, Description="LOCTEXT(\"Match Button clicked\")")
   > n84 = Sequence!(execute) -> (then_0, then_1)
   > n85 = GetFocusedLevelSequence()
   > n86 = Sequence()
   > n87 = Branch(Condition=IsChecked(self=Get CurrentFrameCheckBox()))
   > n88 = Get Current Local Time!(execute, TimeUnit:EMovieSceneTimeUnit Enum=DisplayRate) -> (
    then
    ReturnValue_Frame_FrameNumber_Value:int
    ReturnValue_Frame_SubFrame:float=~
    ReturnValue_Time:float=~
    ReturnValue_MarkedFrame:String
    ReturnValue_Timecode:Timecode Structure
    ReturnValue_PositionType:EMovieScenePositionType Enum
    ReturnValue_UpdateMethod:EUpdatePositionMethod Enum
    ReturnValue_bHasJumped:bool=~
   ) {is=title K2Node_CallFunction}
   > Set!(
    execute
    Playback Range_bHasStart:bool=true
    Playback Range_bHasEnd:bool=true
    Playback Range_InclusiveStart:int=@n88
    Playback Range_ExclusiveEnd:int=@n88
   ) -> (then, Output_Get:Sequencer Scripting Range Structure)
   n89 = Branch(execute=@n87.else, Condition=IsChecked(self=Get SpecifiedRangeCheckBox()))
   n90 = Get Text(self=Get EndFrameTextBox())
   n91 = Get Text(self=Get StartFrameTextBox())
   > Set(
    Playback Range_bHasStart=true
    Playback Range_bHasEnd=true
    Playback Range_InclusiveStart=String To Integer(InString="To String (Text)"(InText=@n91))
    Playback Range_ExclusiveEnd=String To Integer(InString="To String (Text)"(InText=@n90))
   )
   n92 = GetPlaybackRange!(execute=@n89.else, Sequence:Movie Scene Sequence Object Reference=@n85) -> (then, ReturnValue:Sequencer Scripting Range Structure)
   > Set Playback Range(Playback Range:Sequencer Scripting Range Structure=@n92) -> (Output_Get:Sequencer Scripting Range Structure)
   Set Frame Rate(
    execute=@n86.then_1
    Frame Rate:Frame Rate Structure=MakeFrameRate!(Numerator:int=30, Denominator:int=1) -> (ReturnValue:Frame Rate Structure)
   ) -> (Output_Get:Frame Rate Structure)
   > Set Current Matching Frame(Current Matching Frame=Get() -> (Playback Range_bHasStart=~, Playback Range_bHasEnd=~).Playback Range_InclusiveStart) -> (Output_Get=0)
   > n93 = GetControlRigs!(execute, LevelSequence:Level Sequence Object Reference=@n85) -> (then, ReturnValue:Array of Control Rig Sequencer Binding Proxy Structures)
   > n94 = For Each Loop(Array:Array of Control Rig Sequencer Binding Proxy Structures=@n93) -> (Array Element:Control Rig Sequencer Binding Proxy Structure)
   n95 = Break Control Rig Sequencer Binding Proxy!(ControlRigSequencerBindingProxy:"Control Rig Sequencer Binding Proxy Structure (by ref)"=@n94) -> (
    Proxy:Movie Scene Binding Proxy Structure
    ControlRig:Control Rig Object Reference
    Track:Movie Scene Control Rig Parameter Track Object Reference
   ) {is=title K2Node_BreakStruct}
   n96 = Get Primary Control Rig() -> (Primary Control Rig:Control Rig Object Reference)
   > Branch(Condition="Equal (Object)"(A=@n95.ControlRig, B=@n96))
   > Set Track(Track:Movie Scene Control Rig Parameter Track Object Reference=@n95.Track) -> (Output_Get:Movie Scene Control Rig Parameter Track Object Reference)
   > Set Rig Hierarchy(Rig Hierarchy:Rig Hierarchy Object Reference=GetHierarchy!(self:Control Rig Object Reference=@n96) -> (ReturnValue:Rig Hierarchy Object Reference)) -> (Output_Get:Rig Hierarchy Object Reference)
   EndTransaction(execute=@n84.then_1)
   Switch IK FK and Set Key() {is=title K2Node_CustomEvent}
   > Set Output Log Message(Output Message=Get Output Message())
   > Clear(TargetArray:Array of Movie Scene Scripting Key Object References=Get Keys To Delete())
   > n97 = Keys(TargetMap=Get IK FK Switch Channels())
   > n98 = For Each Loop(Array:Array of Movie Scene Scripting Bool Channel Object References=@n97) -> (Array Element:Movie Scene Scripting Bool Channel Object Reference)
   > n99 = Branch(Condition=IsChecked(self=Get CurrentFrameCheckBox()))
   n100 = Make Frame Number!(Value:int=Get() -> (Playback Range_bHasStart=~, Playback Range_bHasEnd=~).Playback Range_InclusiveStart) -> (FrameNumber:Frame Number Structure) {is=title K2Node_MakeStruct}
   n101 = Find(TargetMap=Get IK FK Switch Channels(), Key=@n98)
   > "Add Key (Bool)"!(
    execute
    self:Movie Scene Scripting Bool Channel Object Reference=@n98
    InTime:"Frame Number Structure (by ref)"=@n100
    NewValue:bool=@n101
    SubFrame:float=0
    TimeUnit:EMovieSceneTimeUnit Enum=DisplayRate
   ) -> (then, ReturnValue:Movie Scene Scripting Bool Key Object Reference)
   n102 = For Each Loop(Exec=@n99.else, Array:Array of Movie Scene Scripting Key Object References="Get Keys (Bool)"(self=@n98)) -> (Array Element:Movie Scene Scripting Key Object Reference)
   n103 = For Each Loop(Exec=@n102.Completed, Array=Get Keys To Delete())
   Set Current Local Time!(
    execute=@n103.Completed
    PlaybackParams:Movie Scene Sequence Playback Params Structure=Make Movie Scene Sequence Playback Params!(
     Frame_FrameNumber_Value:int=Get() -> (Playback Range_bHasStart=~, Playback Range_bHasEnd=~).Playback Range_InclusiveStart
     Frame_SubFrame:float=0
     Time:float=0
     MarkedFrame:String
     Timecode:Timecode Structure=(Hours=0,Minutes=0,Seconds=0,Frames=0,Subframe=0,bDropFrameFormat=False)
     PositionType:EMovieScenePositionType Enum=Frame
     UpdateMethod:EUpdatePositionMethod Enum=Play
     bHasJumped:bool=False
    ) -> (MovieSceneSequencePlaybackParams:Movie Scene Sequence Playback Params Structure) {is=title K2Node_MakeStruct}
    TimeUnit=DisplayRate
   ) -> (then)
   > "Add Key (Bool)"(self=@n98, InTime=@n100, NewValue=@n101, SubFrame=0, TimeUnit=DisplayRate)
   "Remove Key (Bool)"(execute=@n103.LoopBody, self=@n98, Key=@n103)
   n104 = Cast To MovieSceneScriptingBoolKey(execute=@n102.LoopBody, Object=@n102)
   n105 = "Get Time (Bool)"(self=@n104, TimeUnit=DisplayRate) -> (ReturnValue_SubFrame=~)
   n106 = Get() -> (Playback Range_bHasStart=~, Playback Range_bHasEnd=~)
   > Branch(Condition=AND Boolean(A="integer >= integer"(A=@n105, B=@n106.Playback Range_InclusiveStart), B="integer <= integer"(A=@n105, B=@n106.Playback Range_ExclusiveEnd)))
   > Add(
    TargetArray:Array of Movie Scene Scripting Key Object References=Get Keys To Delete()
    NewItem:"Movie Scene Scripting Key Object Reference (by ref)"=@n102
   )
   n107 = Set Output Log Message!() -> (OutputDelegate, then, Output Message:String) {is=title K2Node_CustomEvent}
   > SetText(self=Get Output_TextBlock(), InText="To Text (String)"(InString=LeftChop(SourceString=@n107.Output Message, Count=1)))
   > Set Output Message()
   Run Matching On Existing Keys() {is=title K2Node_CustomEvent}
   > n108 = Set Current Exsiting Frame Matching Index(Current Exsiting Frame Matching Index:int=0) -> (Output_Get:int=0)
   n109 = Keys(
    execute=[@n108.then, @n116."  "]
    TargetMap:Map of Names to IKFKMatching Frames Structures=Get Existing Keys Data() -> (Existing Keys Data:Map of Names to IKFKMatching Frames Structures)
   ) -> (Keys:Array of Names)
   > Set Current Existing Key Data Name(Current Existing Key Data Name:Name="Get (a ref)"(Array:Array of Names=@n109, Dimension 1=Get Current Exsiting Frame Matching Index() -> (Current Exsiting Frame Matching Index:int=0)) -> (Output:"Name (by ref)")) -> (Output_Get:Name=None)
   > n110 = Values!(execute, TargetMap:Map of Names to IKFKMatching Frames Structures=Get Existing Keys Data()) -> (then, Values:Array of IKFKMatching Frames Structures)
   n111 = "Get (a ref)"!(
    Array:Array of IKFKMatching Frames Structures=@n110
    Dimension 1=Get Current Exsiting Frame Matching Index() -> (Current Exsiting Frame Matching Index=0)
   ) -> (Output_Frames_4_#7:Array of Integers)
   > n112 = Branch(Condition=Is Not Empty(TargetArray:Array of Integers=@n111))
   > n113 = For Each Loop with Delay!(Exec:Exec, Array:Array of Integers=@n111, Delay:double=0.05) -> (Loop Body:Exec, Array Element:int, Array Index:int, Completed:Exec) {is=title K2Node_MacroInstance}
   n114 = Get Existing Keys Data()
   n115 = Branch(
    execute=[@n113.Completed, @n112.else]
    Condition="integer >= integer"(
     A=Get Current Exsiting Frame Matching Index() -> (Current Exsiting Frame Matching Index=0)
     B=Length!(TargetMap:Map of Names to IKFKMatching Frames Structures=@n114) -> (ReturnValue:int=0) {is=title K2Node_CallFunction}
    )
   )
   > Switch IK FK and Set Key() {is=title K2Node_CallFunction}
   > EndTransaction()
   n116 = Increment Int(" "=@n115.else, Value=Get Current Exsiting Frame Matching Index() -> (Current Exsiting Frame Matching Index=0))
   n117 = Set Current Matching Frame(execute=@n113.Loop Body, Current Matching Frame=@n113) -> (Output_Get=0)
   > Set Current Local Time(
    PlaybackParams=Make Movie Scene Sequence Playback Params!(
     Frame:Frame Time Structure=Make Frame Time!(FrameNumber_Value:int=@n117, SubFrame:float=0) -> (FrameTime:Frame Time Structure) {is=title K2Node_MakeStruct}
     Time=0
     MarkedFrame
     Timecode=(Hours=0,Minutes=0,Seconds=0,Frames=0,Subframe=0,bDropFrameFormat=False)
     PositionType=Frame
     UpdateMethod=Play
     bHasJumped=False
    ) -> (MovieSceneSequencePlaybackParams)
    TimeUnit=DisplayRate
   )
   > DelayUntilNextTick()
   > Run Matching()
   n118 = Event Construct()
   n119 = Cast To ModularRig!(execute=[@n118.then, @n128.then], Object=Get Primary Control Rig()) -> (then, CastFailed, AsModular Rig:Modular Rig Object Reference) {is=title K2Node_DynamicCast}
   > Set IsModular(IsModular:bool=true) -> (Output_Get:bool)
   > n120 = Set Modular Rig(Modular Rig:Modular Rig Object Reference=@n119) -> (Output_Get:Modular Rig Object Reference)
   n121 = Get Animatable Creature()
   n122 = LoadAsset_Blocking(
    execute=[@n129.then, @n120.then]
    Asset=Get IK FK Matching Data!(self:Animatable Creature Object Reference=@n121) -> (IK FK Matching Data:PDA Ik Fk Matching Bake Data Soft Object Reference)
   )
   > n123 = Cast To PDA_IkFkMatchingBakeData!(execute, Object=@n122) -> (then, CastFailed, AsPDA Ik Fk Matching Bake Data:PDA Ik Fk Matching Bake Data Object Reference) {is=title K2Node_DynamicCast}
   n124 = GetDisplayName(Object=@n121)
   > Set Text(Text="To Text (String)"(InString=@n124), self:Text Object Reference=Get AC_TextBlock() -> (AC_TextBlock:Text Object Reference))
   > SetVisibility(self=Get ACBrowse_Button() -> (ACBrowse_Button:Editor Utility Button Object Reference), InVisibility=Visible)
   > Clear(TargetArray:Array of EUW Tolk Fk Check Box Object References=Get Limb Check Boxes())
   > ClearChildren(self=Get Limb_CheckBoxes_VerticalBox() -> (Limb_CheckBoxes_VerticalBox:Vertical Box Object Reference))
   > n125 = Set Ik Fk Matching From Data Asset(Ik Fk Matching From Data Asset:Array of Ik Fk Matching Data Structures=Get Ik Fk Matching Data!(self:PDA Ik Fk Matching Bake Data Object Reference=@n123) -> (Ik Fk Matching Data)) -> (Output_Get:Array of Ik Fk Matching Data Structures)
   > n126 = For Each Loop(Array:Array of Ik Fk Matching Data Structures=@n125) -> (Array Element:Ik Fk Matching Data Structure)
   Set Visibility(execute=@n126.Completed, Visibility=Collapsed, self=Get FrameRateHorizontalBox())
   n127 = Create Widget!(
    execute:Exec=@n126.LoopBody
    WorldContextObject:Object Reference
    Class:User Widget Class Reference=&/EUW_TolkFkCheckBox.EUW_TolkFkCheckBox_C
    OwningPlayer:Player Controller Object Reference
    CheckBoxText:Text="To Text (Name)"!(InName:Name=Break Ik Fk Matching Data(IkFkMatchingData=@n126)) -> (ReturnValue:Text)
   ) -> (then:Exec, ReturnValue:EUW Tolk Fk Check Box Object Reference)
   > AddChild(self=Get Limb_CheckBoxes_VerticalBox(), Content=@n127)
   > Add(
    TargetArray:Array of EUW Tolk Fk Check Box Object References=Get Limb Check Boxes()
    NewItem:"EUW Tolk Fk Check Box Object Reference (by ref)"=@n127
   )
   Set Output Log Message(
    execute=@n123.CastFailed
    Output Message=Append!(A="<Fail>Couldn't load IK FK Matching Data from AC:  ", B=@n124, C="Finding Data from AnimBrowser</>\r\n") -> (ReturnValue)
   ) {is=title K2Node_CallFunction}
   > Do Once()
   > n128 = Fallback_IKFKMatchingDataNotFound()
   n129 = Set IsModular(execute=@n119.CastFailed, IsModular=false)
   Set Primary Control Rig(Primary Control Rig:Control Rig Object Reference) -> (Output_Get:Control Rig Object Reference)
  ]}
 Fallback_IKFKMatchingDataNotFound {nodes [
   Fallback_IKFKMatchingDataNotFound() {is=title K2Node_FunctionEntry}
   > n1 = FindUtilityWidgetFromBlueprint!(
    execute
    self:Editor Utility Subsystem Object Reference=Get EditorUtilitySubsystem()
    InBlueprint:Editor Utility Widget Blueprint Object Reference=/AnimInEdTools/AnimationBrowser/Animation_Browser
   ) -> (then, ReturnValue:Editor Utility Widget Object Reference)
   > n2 = Cast To Animation_Browser!(execute, Object=@n1) -> (then, CastFailed, AsAnimation Browser:Animation Browser Object Reference) {is=title K2Node_DynamicCast}
   > n3 = GetFocusedLevelSequence()
   > n4 = GetControlRigs(LevelSequence=@n3)
   > n5 = For Each Loop(Array:Array of Control Rig Sequencer Binding Proxy Structures=@n4) -> (Array Element:Control Rig Sequencer Binding Proxy Structure)
   n6 = Find Animatable Creature!(
    Asset:String=Get Current Asset!(self:Animation Browser Object Reference=@n2) -> (Current Asset:String)
    Variation:String=Get Current Variation!(self:Animation Browser Object Reference=@n2) -> (Current Variation:String)
    Project:BP Animation Browser Project Object Reference=Get Current Project!(self:Animation Browser Object Reference=@n2) -> (Current Project:BP Animation Browser Project Object Reference)
    __WorldContext:Object Reference
   ) -> (Animatable Creatures:Animatable Creature Object Reference)
   Set Animatable Creature(execute=@n5.Completed, Animatable Creature:Animatable Creature Object Reference=@n6) -> (Output_Get:Animatable Creature Object Reference)
   n7 = Break Control Rig Sequencer Binding Proxy(ControlRigSequencerBindingProxy=@n5)
   n8 = "Get Class Path Name (Top Level Asset Path)"!(Class:Object Class Reference=Get Class(Object=@n7.ControlRig) -> (ReturnValue:Control Rig Class Reference)) -> (ReturnValue_PathString:String) {is=title K2Node_CallFunction}
   n9 = "Get Path Name (Control Rig Soft Reference)"!(InRef:"Control Rig Asset Soft Reference Structure (by ref)"=Get Control Rig!(self:Animatable Creature Object Reference=@n6) -> (Control Rig:Control Rig Asset Soft Reference Structure)) -> (ReturnValue:String) {is=title K2Node_CallFunction}
   Branch(execute=@n5.LoopBody, Condition="Equal Exactly (String)"(A=@n8, B=@n9))
   > PrintString(InString=**** checking asset paths****)
   > PrintString(InString=Append!(A="loopedAsset: ", B=@n8) -> (ReturnValue))
   > PrintString(InString=Append(A="AC file Asset: ", B=@n9))
   > Set Primary Control Rig(Primary Control Rig=@n7.ControlRig)
  ]}
 Find Existing Keys {nodes [
   Find Existing Keys() {is=title K2Node_FunctionEntry}
   > Clear(TargetMap:Map of Names to IKFKMatching Frames Structures=Get Existing Keys Data()) {is=title K2Node_CallFunction}
   > n1 = For Each Loop(Array:Array of EUW Tolk Fk Check Box Object References=Get Limb Check Boxes()) -> (Array Element:EUW Tolk Fk Check Box Object Reference)
   Return Node!(execute=@n1.Completed, Existing Keys Matching Data:Map of Names to IKFKMatching Frames Structures=Get Existing Keys Data())
   > Set Current Matching Data(Current Matching Data:Ik Fk Matching Data Structure="Get (a copy)"(Array=Get Ik Fk Matching From Data Asset(), Dimension 1=@n1.Array Index)) -> (Output_Get:Ik Fk Matching Data Structure)
   n2 = Get ToFk_CheckBox(self=@n1)
   n3 = IsChecked(self=Get ToIk_CheckBox(self=@n1))
   > Branch(Condition=OR Boolean(A=@n3, B=IsChecked(self=@n2)))
   > Set Look for FK(Look for FK:bool=@n3) -> (Output_Get:bool)
   > Clear(TargetArray:Array of Integers=Get Found Frames with Key() -> (Found Frames with Key:Array of Integers)) {is=title K2Node_CallArrayFunction}
   > n4 = GetSections(Track=Get Track())
   > n5 = For Each Loop(Array:Array of Movie Scene Section Object References=@n4) -> (Array Element:Movie Scene Section Object Reference)
   Add(
    execute=@n5.Completed
    TargetMap:Map of Names to IKFKMatching Frames Structures=Get Existing Keys Data()
    Key:"Name (by ref)"=Break Ik Fk Matching Data(IkFkMatchingData=Get Current Matching Data() -> (Current Matching Data:Ik Fk Matching Data Structure))
    Value:"IKFKMatching Frames Structure (by ref)"=Make IKFKMatching Frames!(Frames_4_#7:Array of Integers=Get Found Frames with Key()) -> (IKFKMatchingFrames:IKFKMatching Frames Structure) {is=title K2Node_MakeStruct}
   ) {is=title K2Node_CallFunction}
   n6 = GetAllChannels(execute=@n5.LoopBody, Section=@n5)
   > n7 = For Each Loop(Array:Array of Movie Scene Scripting Channel Object References=@n6) -> (Array Element:Movie Scene Scripting Channel Object Reference)
   > Set Current Channel(Current Channel:Movie Scene Scripting Channel Object Reference=@n7) -> (Output_Get:Movie Scene Scripting Channel Object Reference)
   > Set Found Control(Found Control:bool=false) -> (Output_Get:bool)
   n8 = Break Ik Fk Matching Data(IkFkMatchingData=Get Current Matching Data())
   > n9 = For Each Loop with Break!(
    Exec:Exec
    Array:Array of Names=Select(
     Option 0:Array of Names=@n8.IKControls_10_#5
     Option 1:Array of Names=@n8.FKControls_12_#6
     Index=Get Look for FK() -> (Look for FK:bool)
    ) -> (ReturnValue:Array of Names)
    Break:Exec=@n11.then
   ) -> (LoopBody:Exec, Array Element:Name, Array Index:int, Completed:Exec)
   n10 = GetDisplayName(Object=@n7)
   > Branch(Condition=Contains(SearchIn=Split(SourceString=@n10, InStr=., SearchDir=FromStart), Substring="To String (Name)"(InName=@n9)))
   > n11 = Set Found Control(Found Control=true)
   Branch(execute=@n9.Completed, Condition=Get Found Control() -> (Found Control:bool))
   > n12 = Cast To MovieSceneScriptingFloatChannel(Object=Get Current Channel() -> (Current Channel:Movie Scene Scripting Channel Object Reference))
   > n13 = For Each Loop(Array:Array of Movie Scene Scripting Key Object References="Get Keys (Float)"(self=@n12)) -> (Array Element:Movie Scene Scripting Key Object Reference)
   > n14 = Cast To MovieSceneScriptingFloatKey(Object=@n13)
   > Add Unique(
    TargetArray:Array of Integers=Get Found Frames with Key()
    NewItem:"Integer (by ref)"=Break Frame Time!(FrameTime:"Frame Time Structure (by ref)"="Get Time (Float)"!(self=@n14, TimeUnit=DisplayRate) -> (ReturnValue:Frame Time Structure)) -> (FrameNumber_Value:int, SubFrame:float=~) {is=title K2Node_BreakStruct}
   )
  ]}
 IKFK Match Finished E Dispatcher {nodes [IKFK Match Finished E Dispatcher() {is=title K2Node_FunctionEntry}]}
 Run Matching {nodes [
   Run Matching() {is=title K2Node_FunctionEntry}
   > n1 = For Each Loop(Array:Array of Ik Fk Matching Data Structures=Get Ik Fk Matching Data!() -> (Ik Fk Matching Data)) -> (Array Element:Ik Fk Matching Data Structure)
   > Set Matching Data Index(Matching Data Index:int=@n1.Array Index) -> (Output_Get:int=0)
   n2 = Break Ik Fk Matching Data(IkFkMatchingData=@n1)
   > Set Match IK Event Name(Match IK Event Name:Name=@n2.MatchIKEvent_21_#3) -> (Output_Get:Name=None)
   > Set Match FK Event Name(Match FK Event Name:Name=@n2.MatchFKEvent_22_#4) -> (Output_Get:Name=None)
   > Set Ik Controls(Ik Controls=@n2.IKControls_10_#5)
   > Set Fk Controls(Fk Controls=@n2.FKControls_12_#6)
   n3 = Get IK FK Switch Channels()
   > n4 = Keys(TargetMap:Map of Movie Scene Scripting Bool Channel Object References to Booleans=@n3) -> (Keys:Array of Movie Scene Scripting Bool Channel Object References)
   > n5 = Sequence()
   n6 = Get Matching Data Index() -> (Matching Data Index:int=0)
   > n7 = Branch(Condition=Find(
     TargetMap=@n3
     Key="Get (a copy)"(Array:Array of Movie Scene Scripting Bool Channel Object References=@n4, Dimension 1=@n6) -> (Output:Movie Scene Scripting Bool Channel Object Reference)
    ))
   > n8 = Branch(Condition=Get IsModular())
   n9 = "Get (a ref)"!(Array:Array of Strings=Get Module Paths(), Dimension 1=Get Matching Data Index() -> (Matching Data Index=0)) -> (Output:"String (by ref)")
   > Execute Event On Module By Name!(
    execute
    self:Modular Rig Object Reference=Get Modular Rig()
    InEvent:Name=Get Match IK Event Name() -> (Match IK Event Name:Name=None)
    InModuleName:Name=String To Name(InString=@n9)
   ) -> (then, ReturnValue:bool)
   > n10 = Add Module Path To Control Names(Module Path=@n9, Controls=Get Ik Controls(), Only Module Name=false)
   > Set Controls To Key(Controls To Key=@n10)
   > Execute!(execute, self:Rig VMHost Object Reference=Get Modular Rig(), InEventName:"Name (by ref)"=MakeLiteralName(Value=Forwards Solve)) -> (then, ReturnValue:bool)
   ExecuteEvent!(
    execute=@n8.else
    self:Rig VMHost Object Reference=Get Primary Control Rig()
    InEventName:"Name (by ref)"=Get Match IK Event Name() -> (Match IK Event Name=None)
   ) -> (then, ReturnValue:bool)
   > Set Controls To Key(Controls To Key=Get Ik Controls())
   n11 = Branch(execute=@n7.else, Condition=Get IsModular())
   n12 = "Get (a ref)"(Array=Get Module Paths(), Dimension 1=Get Matching Data Index() -> (Matching Data Index=0))
   > Execute Event On Module By Name(self=Get Modular Rig(), InEvent=Get Match FK Event Name() -> (Match FK Event Name:Name=None), InModuleName=String To Name(InString=@n12))
   > n13 = Add Module Path To Control Names(Module Path=@n12, Controls=Get Fk Controls(), Only Module Name=false)
   > Set Controls To Key(Controls To Key=@n13)
   > Execute(self=Get Modular Rig(), InEventName=MakeLiteralName(Value=Forwards Solve))
   ExecuteEvent(execute=@n11.else, self=Get Primary Control Rig(), InEventName=Get Match FK Event Name() -> (Match FK Event Name=None))
   > Set Controls To Key(Controls To Key=Get Fk Controls())
   n14 = For Each Loop(Exec=@n5.then_1, Array:Array of Names=Get Controls To Key()) -> (Array Element:Name)
   > SendAutoKeyEvent!(
    execute
    self:Rig Hierarchy Object Reference=Get Rig Hierarchy() -> (Rig Hierarchy:Rig Hierarchy Object Reference)
    InElement_Type:ERigElementType Enum=Control
    InElement_Name:Name=@n14
    InOffsetInSeconds:float=0
    bAsynchronous:bool=false
   ) -> (then)
  ]}
]
variables [
 Animatable Creature {type=AnimatableCreature_C, default=null}
 Controls To Key {type="array<string>", default []}
 Current Existing Key Data Name {type=string, default=None}
 Current Existing Key IKFK Matching Data {type=IkFkMatchingData, default {name=None, ik Fk Switch Attr=None, match FK Event=None, match IK Event=None, fK Controls [], iK Controls []}}
 Current Exsiting Frame Matching Index {type=integer}
 Current Matching Frame {type=integer}
 Existing Keys Data {type="map<IKFKMatchingFrames>", default {}}
 Fk Controls {type="array<string>", default []}
 Frame Rate {type=FrameRate, default {denominator=0, numerator=0}}
 Ik Controls {type="array<string>", default []}
 Ik Fk Matching Data {type="array<IkFkMatchingData>", default []}
 Ik Fk Matching From Data Asset {type="array<IkFkMatchingData>", default []}
 IK FK Switch Channels {default {}}
 IK FK Switch Values {type="array<boolean>", default []}
 IKFK Match Finished E Dispatcher {default=()}
 Is IK {type=boolean}
 IsModular {type=boolean}
 Keys To Delete {type="array<MovieSceneScriptingKey>", default []}
 Limb Check Boxes {type="array<EUW_TolkFkCheckBox_C>", default []}
 Match FK Event Name {type=string, default=None}
 Match IK Event Name {type=string, default=None}
 Matching Data Index {type=integer}
 Modular Rig {type=ModularRig, default=null}
 Module Paths {type="array<string>", default []}
 Output Message {type=string, default=""}
 Playback Range {
  type=SequencerScriptingRange
  default {bHasEnd=false, bHasStart=false, exclusiveEnd=0, inclusiveStart=0, internalRate {denominator=1, numerator=60000}}
 }
 Primary Control Rig {type=ControlRig, default=null}
 Range {
  type=SequencerScriptingRange
  default {bHasEnd=false, bHasStart=false, exclusiveEnd=0, inclusiveStart=0, internalRate {denominator=1, numerator=60000}}
 }
 Rig Hierarchy {type=RigHierarchy, default=null}
 Track {type=MovieSceneControlRigParameterTrack, default=null}
]
widgets {
 info {namedSlotCount=8, rootWidgetClass=&VerticalBox, widgetCount=45}
 widgets [
  VerticalBox_18 {widgetClassPath=&VerticalBox, widget=VerticalBox_18}
  HorizontalBox_0 {
   widgetClassPath=&HorizontalBox
   parent=VerticalBox_18
   widget=HorizontalBox_0
   slot {class=VerticalBoxSlot, properties {padding=<ltrb 2 0 0 0>}}
  }
  AC_TextBlock {
   widgetClassPath=&TextBlock
   bIsVariable=true
   parent=HorizontalBox_0
   widget=AC_TextBlock
   properties {
    justification=Center
    colorAndOpacity {colorUseRule=UseColor_Specified, specifiedColor=<rgba 0.3372 0.3372 0.3372 1>}
    font {
     bForceMonospaced=false
     bMaterialIsStencil=false
     fontMaterial=null
     fontObject=&Roboto
     letterSpacing=0
     monospacedWidth=1
     size=11.25
     skewAmount=0
     typefaceFontName=Regular
     outlineSettings {
      bApplyOutlineToDropShadows=false
      bMiteredCorners=false
      bSeparateFillAlpha=false
      outlineMaterial=null
      outlineSize=0
      outlineColor=<rgba 0 0 0 1>
     }
    }
    strikeBrush {
     bIsDynamicallyLoaded=false
     drawAs=NoDrawType
     imageType=NoImage
     mirroring=NoMirror
     resourceName=None
     resourceObject=null
     tiling=NoTile
     imageSize=<32 32>
     margin=<ltrb 0 0 1 1>
     outlineSettings {
      bUseBrushTransparency=false
      roundingType=HalfHeightRadius
      width=0
      color {colorUseRule=UseColor_Specified, specifiedColor=<rgba 0 0 0 0>}
      cornerRadii=<0 0 0 0>
     }
     tintColor {colorUseRule=UseColor_Specified, specifiedColor=<rgba 0 0 0 1>}
     uVRegion {bIsValid=false, max=<0 0>, min=<0 0>}
    }
   }
   slot {class=HorizontalBoxSlot, properties {horizontalAlignment=HAlign_Center, verticalAlignment=VAlign_Center, padding=<ltrb 4 2 4 2>}}
  }
  ACBrowse_Button {
   widgetClassPath=&EditorUtilityButton
   parent=HorizontalBox_0
   widget=ACBrowse_Button
   properties {
    toolTipText=Browse to Canvas
    visibility=Hidden
    widgetStyle {
     clickedSlateSound {resourceObject=null}
     disabled {
      bIsDynamicallyLoaded=false
      drawAs=Image
      imageType=FullColor
      mirroring=NoMirror
      resourceName=None
      resourceObject=/Game/UI/Foundation/Textures/Icons/HUD/T-IconMagnifying-Glass-32
      tiling=NoTile
      imageSize=<15 15>
      margin=<ltrb 0 0 0 0>
      outlineSettings {
       bUseBrushTransparency=false
       roundingType=FixedRadius
       width=1
       color {colorUseRule=UseColor_ColorTable, specifiedColor=<rgba 0 0 0 0>}
       cornerRadii=<4 4 4 4>
      }
      tintColor {colorUseRule=UseColor_Specified, specifiedColor=<rgba 0.0395 0.0395 0.0395 1>}
      uVRegion {bIsValid=false, max=<0 0>, min=<0 0>}
     }
     disabledForeground {colorUseRule=UseColor_ColorTable, specifiedColor=<rgba 0 0 0 0>}
     hovered {
      bIsDynamicallyLoaded=false
      drawAs=Image
      imageType=FullColor
      mirroring=NoMirror
      resourceName=None
      resourceObject=&T-IconMagnifying-Glass-32
      tiling=NoTile
      imageSize=<15 15>
      margin=<ltrb 0 0 0 0>
      outlineSettings {
       bUseBrushTransparency=false
       roundingType=FixedRadius
       width=1
       color {colorUseRule=UseColor_ColorTable, specifiedColor=<rgba 0 0 0 0>}
       cornerRadii=<4 4 4 4>
      }
      tintColor {colorUseRule=UseColor_Specified, specifiedColor=<rgba 0.9319 0.9319 0.9319 1>}
      uVRegion {bIsValid=false, max=<0 0>, min=<0 0>}
     }
     hoveredForeground {colorUseRule=UseColor_ColorTable, specifiedColor=<rgba 0 0 0 0>}
     hoveredSlateSound {resourceObject=null}
     normal {
      bIsDynamicallyLoaded=false
      drawAs=Image
      imageType=FullColor
      mirroring=NoMirror
      resourceName=None
      resourceObject=&T-IconMagnifying-Glass-32
      tiling=NoTile
      imageSize=<15 15>
      margin=<ltrb 0 0 0 0>
      outlineSettings {
       bUseBrushTransparency=false
       roundingType=FixedRadius
       width=1
       color {colorUseRule=UseColor_ColorTable, specifiedColor=<rgba 0 0 0 0>}
       cornerRadii=<4 4 4 4>
      }
      tintColor {colorUseRule=UseColor_Specified, specifiedColor=<rgba 0.2274 0.2274 0.2274 1>}
      uVRegion {bIsValid=false, max=<0 0>, min=<0 0>}
     }
     normalForeground {colorUseRule=UseColor_ColorTable, specifiedColor=<rgba 0 0 0 0>}
     normalPadding=<ltrb 12 1.5 12 1.5>
     pressed {
      bIsDynamicallyLoaded=false
      drawAs=Image
      imageType=FullColor
      mirroring=NoMirror
      resourceName=None
      resourceObject=&T-IconMagnifying-Glass-32
      tiling=NoTile
      imageSize=<15 15>
      margin=<ltrb 0 0 0 0>
      outlineSettings {
       bUseBrushTransparency=false
       roundingType=FixedRadius
       width=1
       color {colorUseRule=UseColor_ColorTable, specifiedColor=<rgba 0 0 0 0>}
       cornerRadii=<4 4 4 4>
      }
      tintColor {colorUseRule=UseColor_Specified, specifiedColor=<rgba 0.0284 0.0284 0.0284 1>}
      uVRegion {bIsValid=false, max=<0 0>, min=<0 0>}
     }
     pressedForeground {colorUseRule=UseColor_ColorTable, specifiedColor=<rgba 0 0 0 0>}
     pressedPadding=<ltrb 12 2.5 12 0.5>
     pressedSlateSound {resourceObject=null}
    }
   }
   slot {class=HorizontalBoxSlot, properties {horizontalAlignment=HAlign_Center, verticalAlignment=VAlign_Center}}
  }
  Spacer_131 {
   widgetClassPath=&Spacer
   parent=HorizontalBox_0
   widget=Spacer_131
   slot {class=HorizontalBoxSlot, properties {size {sizeRule=Fill, value=1}}}
  }
  Limbs_EditorUtilityExpandableArea {
   widgetClassPath=/Script/Blutility.EditorUtilityExpandableArea
   bIsVariable=true
   parent=VerticalBox_18
   widget=Limbs_EditorUtilityExpandableArea
   properties {
    bIsExpanded=true
    borderBrush {
     bIsDynamicallyLoaded=false
     drawAs=RoundedBox
     imageType=NoImage
     mirroring=NoMirror
     resourceName=None
     resourceObject=null
     tiling=NoTile
     imageSize=<0 0>
     margin=<ltrb 0 0 0 0>
     outlineSettings {
      bUseBrushTransparency=false
      roundingType=FixedRadius
      width=0
      color {colorUseRule=UseColor_Specified, specifiedColor=<rgba 0 0 0 0>}
      cornerRadii=<4 4 4 4>
     }
     tintColor {colorUseRule=UseColor_Specified, specifiedColor=<rgba 0 0 0 1>}
     uVRegion {bIsValid=false, max=<0 0>, min=<0 0>}
    }
   }
   slot {class=VerticalBoxSlot}
  }
  TextBlock_93 {
   widgetClassPath=&TextBlock
   namedSlotHost=Limbs_EditorUtilityExpandableArea
   widget=TextBlock_93
   properties {
    text=Limbs
    font {
     bForceMonospaced=false
     bMaterialIsStencil=false
     fontMaterial=null
     fontObject=&Roboto
     letterSpacing=0
     monospacedWidth=1
     size=15
     skewAmount=0
     typefaceFontName=Italic
     outlineSettings {
      bApplyOutlineToDropShadows=false
      bMiteredCorners=false
      bSeparateFillAlpha=false
      outlineMaterial=null
      outlineSize=0
      outlineColor=<rgba 0 0 0 1>
     }
    }
   }
  }
  VerticalBox_0 {widgetClassPath=&VerticalBox, namedSlotHost=Limbs_EditorUtilityExpandableArea, widget=VerticalBox_0}
  Limb_CheckBoxes_VerticalBox {
   widgetClassPath=&VerticalBox
   bIsVariable=true
   parent=VerticalBox_0
   widget=Limb_CheckBoxes_VerticalBox
   slot {class=VerticalBoxSlot}
  }
  FrameRate_EditorUtilityExpandableArea {
   widgetClassPath=&EditorUtilityExpandableArea
   bIsVariable=true
   parent=VerticalBox_18
   widget=FrameRate_EditorUtilityExpandableArea
   properties {
    bIsExpanded=true
    borderBrush {
     bIsDynamicallyLoaded=false
     drawAs=RoundedBox
     imageType=NoImage
     mirroring=NoMirror
     resourceName=None
     resourceObject=null
     tiling=NoTile
     imageSize=<0 0>
     margin=<ltrb 0 0 0 0>
     outlineSettings {
      bUseBrushTransparency=false
      roundingType=FixedRadius
      width=0
      color {colorUseRule=UseColor_Specified, specifiedColor=<rgba 0 0 0 0>}
      cornerRadii=<4 4 4 4>
     }
     tintColor {colorUseRule=UseColor_Specified, specifiedColor=<rgba 0 0 0 1>}
     uVRegion {bIsValid=false, max=<0 0>, min=<0 0>}
    }
   }
   slot {class=VerticalBoxSlot}
  }
  TextBlock {
   widgetClassPath=&TextBlock
   namedSlotHost=FrameRate_EditorUtilityExpandableArea
   widget=TextBlock
   properties {
    text=Frame Range
    font {
     bForceMonospaced=false
     bMaterialIsStencil=false
     fontMaterial=null
     fontObject=&Roboto
     letterSpacing=0
     monospacedWidth=1
     size=15
     skewAmount=0
     typefaceFontName=Italic
     outlineSettings {
      bApplyOutlineToDropShadows=false
      bMiteredCorners=false
      bSeparateFillAlpha=false
      outlineMaterial=null
      outlineSize=0
      outlineColor=<rgba 0 0 0 1>
     }
    }
   }
  }
  VerticalBox_137 {widgetClassPath=&VerticalBox, namedSlotHost=FrameRate_EditorUtilityExpandableArea, widget=VerticalBox_137}
  HorizontalBox_2 {
   widgetClassPath=&HorizontalBox
   parent=VerticalBox_137
   widget=HorizontalBox_2
   slot {class=VerticalBoxSlot, properties {horizontalAlignment=HAlign_Center, size {sizeRule=Fill, value=1}}}
  }
  TimeSliderCheckBox {
   widgetClassPath=&EditorUtilityCheckBox
   bIsVariable=true
   parent=HorizontalBox_2
   widget=TimeSliderCheckBox
   properties {checkedState=Checked, horizontalAlignment=HAlign_Center}
   slot {class=HorizontalBoxSlot, properties {horizontalAlignment=HAlign_Center, padding=<ltrb 8 8 8 8>}}
  }
  LegR_TextBlock_2 {
   widgetClassPath=&TextBlock
   parent=TimeSliderCheckBox
   widget=LegR_TextBlock_2
   properties {
    text=Time Slider
    font {
     bForceMonospaced=false
     bMaterialIsStencil=false
     fontMaterial=null
     fontObject=&Roboto
     letterSpacing=0
     monospacedWidth=1
     size=13.5
     skewAmount=0
     typefaceFontName=Regular
     outlineSettings {
      bApplyOutlineToDropShadows=false
      bMiteredCorners=false
      bSeparateFillAlpha=false
      outlineMaterial=null
      outlineSize=0
      outlineColor=<rgba 0 0 0 1>
     }
    }
   }
   slot {class=PanelSlot}
  }
  CurrentFrameCheckBox {
   widgetClassPath=&EditorUtilityCheckBox
   bIsVariable=true
   parent=HorizontalBox_2
   widget=CurrentFrameCheckBox
   slot {class=HorizontalBoxSlot, properties {padding=<ltrb 5 5 5 5>}}
  }
  CurrentFrame_TextBlock {
   widgetClassPath=&TextBlock
   parent=CurrentFrameCheckBox
   widget=CurrentFrame_TextBlock
   properties {
    text=Current Frame
    font {
     bForceMonospaced=false
     bMaterialIsStencil=false
     fontMaterial=null
     fontObject=&Roboto
     letterSpacing=0
     monospacedWidth=1
     size=13.5
     skewAmount=0
     typefaceFontName=Regular
     outlineSettings {
      bApplyOutlineToDropShadows=false
      bMiteredCorners=false
      bSeparateFillAlpha=false
      outlineMaterial=null
      outlineSize=0
      outlineColor=<rgba 0 0 0 1>
     }
    }
   }
   slot {class=PanelSlot}
  }
  SpecifiedRangeCheckBox {
   widgetClassPath=&EditorUtilityCheckBox
   bIsVariable=true
   parent=HorizontalBox_2
   widget=SpecifiedRangeCheckBox
   properties {horizontalAlignment=HAlign_Center}
   slot {class=HorizontalBoxSlot, properties {horizontalAlignment=HAlign_Center, padding=<ltrb 8 8 8 8>}}
  }
  LegR_TextBlock {
   widgetClassPath=&TextBlock
   parent=SpecifiedRangeCheckBox
   widget=LegR_TextBlock
   properties {
    text=Specified Range
    font {
     bForceMonospaced=false
     bMaterialIsStencil=false
     fontMaterial=null
     fontObject=&Roboto
     letterSpacing=0
     monospacedWidth=1
     size=13.5
     skewAmount=0
     typefaceFontName=Regular
     outlineSettings {
      bApplyOutlineToDropShadows=false
      bMiteredCorners=false
      bSeparateFillAlpha=false
      outlineMaterial=null
      outlineSize=0
      outlineColor=<rgba 0 0 0 1>
     }
    }
   }
   slot {class=PanelSlot}
  }
  FrameRateHorizontalBox {
   widgetClassPath=&HorizontalBox
   bIsVariable=true
   parent=VerticalBox_137
   widget=FrameRateHorizontalBox
   slot {class=VerticalBoxSlot}
  }
  TextBlock_136 {
   widgetClassPath=&TextBlock
   parent=FrameRateHorizontalBox
   widget=TextBlock_136
   properties {
    text=Start Frame:
    font {
     bForceMonospaced=false
     bMaterialIsStencil=false
     fontMaterial=null
     fontObject=&Roboto
     letterSpacing=0
     monospacedWidth=1
     size=10.5
     skewAmount=0
     typefaceFontName=Regular
     outlineSettings {
      bApplyOutlineToDropShadows=false
      bMiteredCorners=false
      bSeparateFillAlpha=false
      outlineMaterial=null
      outlineSize=0
      outlineColor=<rgba 0 0 0 1>
     }
    }
   }
   slot {class=HorizontalBoxSlot, properties {padding=<ltrb 10 10 10 10>}}
  }
  StartFrameTextBox {
   widgetClassPath=&EditorUtilityEditableTextBox
   bIsVariable=true
   parent=FrameRateHorizontalBox
   widget=StartFrameTextBox
   properties {clearKeyboardFocusOnCommit=false, hintText=0, selectAllTextWhenFocused=true}
   slot {class=HorizontalBoxSlot, properties {padding=<ltrb 0 10 10 10>, size {sizeRule=Fill, value=1}}}
  }
  TextBlock_2 {
   widgetClassPath=&TextBlock
   parent=FrameRateHorizontalBox
   widget=TextBlock_2
   properties {
    text=End Frame:
    font {
     bForceMonospaced=false
     bMaterialIsStencil=false
     fontMaterial=null
     fontObject=&Roboto
     letterSpacing=0
     monospacedWidth=1
     size=10.5
     skewAmount=0
     typefaceFontName=Regular
     outlineSettings {
      bApplyOutlineToDropShadows=false
      bMiteredCorners=false
      bSeparateFillAlpha=false
      outlineMaterial=null
      outlineSize=0
      outlineColor=<rgba 0 0 0 1>
     }
    }
   }
   slot {class=HorizontalBoxSlot, properties {padding=<ltrb 10 10 10 10>}}
  }
  EndFrameTextBox {
   widgetClassPath=&EditorUtilityEditableTextBox
   bIsVariable=true
   parent=FrameRateHorizontalBox
   widget=EndFrameTextBox
   properties {clearKeyboardFocusOnCommit=false, hintText=100, selectAllTextWhenFocused=true}
   slot {class=HorizontalBoxSlot, properties {padding=<ltrb 0 10 5 10>, size {sizeRule=Fill, value=1}}}
  }
  SizeBox_44 {
   widgetClassPath=&SizeBox
   parent=FrameRateHorizontalBox
   widget=SizeBox_44
   properties {bOverride_HeightOverride=true, bOverride_WidthOverride=true, heightOverride=25, widthOverride=25}
   slot {class=HorizontalBoxSlot, properties {horizontalAlignment=HAlign_Left, verticalAlignment=VAlign_Center}}
  }
  SetFrameRangeFromSelected_Button {
   widgetClassPath=&EditorUtilityButton
   parent=SizeBox_44
   widget=SetFrameRangeFromSelected_Button
   properties {
    toolTipText=Set Frame Range from Selected Keys
    widgetStyle {
     clickedSlateSound {resourceObject=null}
     disabled {
      bIsDynamicallyLoaded=false
      drawAs=RoundedBox
      imageType=NoImage
      mirroring=NoMirror
      resourceName=None
      resourceObject=null
      tiling=NoTile
      imageSize=<0 0>
      margin=<ltrb 0 0 0 0>
      outlineSettings {
       bUseBrushTransparency=false
       roundingType=FixedRadius
       width=1
       color {colorUseRule=UseColor_ColorTable, specifiedColor=<rgba 0 0 0 0>}
       cornerRadii=<4 4 4 4>
      }
      tintColor {colorUseRule=UseColor_Specified, specifiedColor=<rgba 0.0395 0.0395 0.0395 1>}
      uVRegion {bIsValid=false, max=<0 0>, min=<0 0>}
     }
     disabledForeground {colorUseRule=UseColor_ColorTable, specifiedColor=<rgba 0 0 0 0>}
     hovered {
      bIsDynamicallyLoaded=false
      drawAs=RoundedBox
      imageType=NoImage
      mirroring=NoMirror
      resourceName=None
      resourceObject=null
      tiling=NoTile
      imageSize=<32 32>
      margin=<ltrb 0 0 0 0>
      outlineSettings {
       bUseBrushTransparency=false
       roundingType=FixedRadius
       width=1
       color {colorUseRule=UseColor_ColorTable, specifiedColor=<rgba 0 0 0 0>}
       cornerRadii=<4 4 4 4>
      }
      tintColor {colorUseRule=UseColor_Specified, specifiedColor=<rgba 0.0953 0.0953 0.0953 1>}
      uVRegion {bIsValid=false, max=<0 0>, min=<0 0>}
     }
     hoveredForeground {colorUseRule=UseColor_ColorTable, specifiedColor=<rgba 0 0 0 0>}
     hoveredSlateSound {resourceObject=null}
     normal {
      bIsDynamicallyLoaded=false
      drawAs=RoundedBox
      imageType=FullColor
      mirroring=NoMirror
      resourceName=None
      resourceObject=/Game/UI/Style/Button-BasicArrowLeft-_normal_
      tiling=NoTile
      imageSize=<24 48>
      margin=<ltrb 0 0 0 0>
      outlineSettings {
       bUseBrushTransparency=false
       roundingType=FixedRadius
       width=1
       color {colorUseRule=UseColor_ColorTable, specifiedColor=<rgba 0 0 0 0>}
       cornerRadii=<4 4 4 4>
      }
      tintColor {colorUseRule=UseColor_Specified, specifiedColor=<rgba 0.5183 0.5183 0.5183 1>}
      uVRegion {bIsValid=false, max=<0 0>, min=<0 0>}
     }
     normalForeground {colorUseRule=UseColor_ColorTable, specifiedColor=<rgba 0 0 0 0>}
     normalPadding=<ltrb 12 1.5 12 1.5>
     pressed {
      bIsDynamicallyLoaded=false
      drawAs=RoundedBox
      imageType=NoImage
      mirroring=NoMirror
      resourceName=None
      resourceObject=null
      tiling=NoTile
      imageSize=<32 32>
      margin=<ltrb 0 0 0 0>
      outlineSettings {
       bUseBrushTransparency=false
       roundingType=FixedRadius
       width=1
       color {colorUseRule=UseColor_ColorTable, specifiedColor=<rgba 0 0 0 0>}
       cornerRadii=<4 4 4 4>
      }
      tintColor {colorUseRule=UseColor_Specified, specifiedColor=<rgba 0.0284 0.0284 0.0284 1>}
      uVRegion {bIsValid=false, max=<0 0>, min=<0 0>}
     }
     pressedForeground {colorUseRule=UseColor_ColorTable, specifiedColor=<rgba 0 0 0 0>}
     pressedPadding=<ltrb 12 2.5 12 0.5>
     pressedSlateSound {resourceObject=null}
    }
   }
   slot {class=SizeBoxSlot}
  }
  BakeSettings_EditorUtilityExpandableArea {
   widgetClassPath=&EditorUtilityExpandableArea
   bIsVariable=true
   parent=VerticalBox_18
   widget=BakeSettings_EditorUtilityExpandableArea
   properties {
    bIsExpanded=true
    borderBrush {
     bIsDynamicallyLoaded=false
     drawAs=RoundedBox
     imageType=NoImage
     mirroring=NoMirror
     resourceName=None
     resourceObject=null
     tiling=NoTile
     imageSize=<0 0>
     margin=<ltrb 0 0 0 0>
     outlineSettings {
      bUseBrushTransparency=false
      roundingType=FixedRadius
      width=0
      color {colorUseRule=UseColor_Specified, specifiedColor=<rgba 0 0 0 0>}
      cornerRadii=<4 4 4 4>
     }
     tintColor {colorUseRule=UseColor_Specified, specifiedColor=<rgba 0 0 0 1>}
     uVRegion {bIsValid=false, max=<0 0>, min=<0 0>}
    }
   }
   slot {class=VerticalBoxSlot}
  }
  TextBlock_1 {
   widgetClassPath=&TextBlock
   namedSlotHost=BakeSettings_EditorUtilityExpandableArea
   widget=TextBlock_1
   properties {
    text=Bake Settings
    font {
     bForceMonospaced=false
     bMaterialIsStencil=false
     fontMaterial=null
     fontObject=&Roboto
     letterSpacing=0
     monospacedWidth=1
     size=15
     skewAmount=0
     typefaceFontName=Italic
     outlineSettings {
      bApplyOutlineToDropShadows=false
      bMiteredCorners=false
      bSeparateFillAlpha=false
      outlineMaterial=null
      outlineSize=0
      outlineColor=<rgba 0 0 0 1>
     }
    }
   }
  }
  VerticalBox_28 {widgetClassPath=&VerticalBox, namedSlotHost=BakeSettings_EditorUtilityExpandableArea, widget=VerticalBox_28}
  HorizontalBox_1 {
   widgetClassPath=&HorizontalBox
   parent=VerticalBox_28
   widget=HorizontalBox_1
   slot {class=VerticalBoxSlot, properties {horizontalAlignment=HAlign_Center, size {sizeRule=Fill, value=1}}}
  }
  CurrentFrame_TextBlock_1 {
   widgetClassPath=&TextBlock
   parent=HorizontalBox_1
   widget=CurrentFrame_TextBlock_1
   properties {
    text=Key every
    font {
     bForceMonospaced=false
     bMaterialIsStencil=false
     fontMaterial=null
     fontObject=&Roboto
     letterSpacing=0
     monospacedWidth=1
     size=13.5
     skewAmount=0
     typefaceFontName=Regular
     outlineSettings {
      bApplyOutlineToDropShadows=false
      bMiteredCorners=false
      bSeparateFillAlpha=false
      outlineMaterial=null
      outlineSize=0
      outlineColor=<rgba 0 0 0 1>
     }
    }
   }
   slot {class=HorizontalBoxSlot, properties {padding=<ltrb 0 10 10 0>}}
  }
  KeyEveryTextBox {
   widgetClassPath=&EditorUtilityEditableTextBox
   bIsVariable=true
   parent=HorizontalBox_1
   widget=KeyEveryTextBox
   properties {hintText=0, text=1}
   slot {class=HorizontalBoxSlot, properties {padding=<ltrb 0 10 10 10>, size {sizeRule=Fill, value=1}}}
  }
  CurrentKeysOnlyCheckBox {
   widgetClassPath=&EditorUtilityCheckBox
   bIsVariable=true
   parent=HorizontalBox_1
   widget=CurrentKeysOnlyCheckBox
   properties {visibility=Hidden}
   slot {class=HorizontalBoxSlot, properties {horizontalAlignment=HAlign_Center, verticalAlignment=VAlign_Center, padding=<ltrb 5 5 5 5>}}
  }
  CurrentKeys_TextBlock {
   widgetClassPath=&TextBlock
   parent=CurrentKeysOnlyCheckBox
   widget=CurrentKeys_TextBlock
   properties {
    text=Existing Keys Only
    font {
     bForceMonospaced=false
     bMaterialIsStencil=false
     fontMaterial=null
     fontObject=&Roboto
     letterSpacing=0
     monospacedWidth=1
     size=13.5
     skewAmount=0
     typefaceFontName=Regular
     outlineSettings {
      bApplyOutlineToDropShadows=false
      bMiteredCorners=false
      bSeparateFillAlpha=false
      outlineMaterial=null
      outlineSize=0
      outlineColor=<rgba 0 0 0 1>
     }
    }
   }
   slot {class=PanelSlot}
  }
  HorizontalBox_43 {widgetClassPath=&HorizontalBox, parent=VerticalBox_18, widget=HorizontalBox_43, slot {class=VerticalBoxSlot}}
  Match_Button {
   widgetClassPath=&EditorUtilityButton
   parent=HorizontalBox_43
   widget=Match_Button
   properties {widgetStyle {
     clickedSlateSound {resourceObject=null}
     disabled {
      bIsDynamicallyLoaded=false
      drawAs=RoundedBox
      imageType=NoImage
      mirroring=NoMirror
      resourceName=None
      resourceObject=null
      tiling=NoTile
      imageSize=<0 0>
      margin=<ltrb 0 0 0 0>
      outlineSettings {
       bUseBrushTransparency=false
       roundingType=FixedRadius
       width=1
       color {colorUseRule=UseColor_ColorTable, specifiedColor=<rgba 0 0 0 0>}
       cornerRadii=<4 4 4 4>
      }
      tintColor {colorUseRule=UseColor_Specified, specifiedColor=<rgba 0.0395 0.0395 0.0395 1>}
      uVRegion {bIsValid=false, max=<0 0>, min=<0 0>}
     }
     disabledForeground {colorUseRule=UseColor_ColorTable, specifiedColor=<rgba 0 0 0 0>}
     hovered {
      bIsDynamicallyLoaded=false
      drawAs=RoundedBox
      imageType=NoImage
      mirroring=NoMirror
      resourceName=None
      resourceObject=null
      tiling=NoTile
      imageSize=<32 32>
      margin=<ltrb 0 0 0 0>
      outlineSettings {
       bUseBrushTransparency=false
       roundingType=FixedRadius
       width=1
       color {colorUseRule=UseColor_ColorTable, specifiedColor=<rgba 0 0 0 0>}
       cornerRadii=<4 4 4 4>
      }
      tintColor {colorUseRule=UseColor_Specified, specifiedColor=<rgba 0.0953 0.0953 0.0953 1>}
      uVRegion {bIsValid=false, max=<0 0>, min=<0 0>}
     }
     hoveredForeground {colorUseRule=UseColor_ColorTable, specifiedColor=<rgba 0 0 0 0>}
     hoveredSlateSound {resourceObject=null}
     normal {
      bIsDynamicallyLoaded=false
      drawAs=RoundedBox
      imageType=NoImage
      mirroring=NoMirror
      resourceName=None
      resourceObject=null
      tiling=NoTile
      imageSize=<32 32>
      margin=<ltrb 0 0 0 0>
      outlineSettings {
       bUseBrushTransparency=false
       roundingType=FixedRadius
       width=1
       color {colorUseRule=UseColor_ColorTable, specifiedColor=<rgba 0 0 0 0>}
       cornerRadii=<4 4 4 4>
      }
      tintColor {colorUseRule=UseColor_Specified, specifiedColor=<rgba 0.0179 0.2887 0 1>}
      uVRegion {bIsValid=false, max=<0 0>, min=<0 0>}
     }
     normalForeground {colorUseRule=UseColor_ColorTable, specifiedColor=<rgba 0 0 0 0>}
     normalPadding=<ltrb 12 1.5 12 1.5>
     pressed {
      bIsDynamicallyLoaded=false
      drawAs=RoundedBox
      imageType=NoImage
      mirroring=NoMirror
      resourceName=None
      resourceObject=null
      tiling=NoTile
      imageSize=<32 32>
      margin=<ltrb 0 0 0 0>
      outlineSettings {
       bUseBrushTransparency=false
       roundingType=FixedRadius
       width=1
       color {colorUseRule=UseColor_ColorTable, specifiedColor=<rgba 0 0 0 0>}
       cornerRadii=<4 4 4 4>
      }
      tintColor {colorUseRule=UseColor_Specified, specifiedColor=<rgba 0.0284 0.0284 0.0284 1>}
      uVRegion {bIsValid=false, max=<0 0>, min=<0 0>}
     }
     pressedForeground {colorUseRule=UseColor_ColorTable, specifiedColor=<rgba 0 0 0 0>}
     pressedPadding=<ltrb 12 2.5 12 0.5>
     pressedSlateSound {resourceObject=null}
    }}
   slot {class=HorizontalBoxSlot, properties {size {sizeRule=Fill, value=1}}}
  }
  TextBlock_53 {
   widgetClassPath=&TextBlock
   parent=Match_Button
   widget=TextBlock_53
   properties {
    text="Match!"
    font {
     bForceMonospaced=false
     bMaterialIsStencil=false
     fontMaterial=null
     fontObject=&Roboto
     letterSpacing=0
     monospacedWidth=1
     size=15
     skewAmount=0
     typefaceFontName=Bold
     outlineSettings {
      bApplyOutlineToDropShadows=false
      bMiteredCorners=false
      bSeparateFillAlpha=false
      outlineMaterial=null
      outlineSize=0
      outlineColor=<rgba 0 0 0 1>
     }
    }
   }
   slot {class=ButtonSlot}
  }
  HorizontalBox {widgetClassPath=&HorizontalBox, parent=VerticalBox_18, widget=HorizontalBox, slot {class=VerticalBoxSlot}}
  Cancel_Button {
   widgetClassPath=&EditorUtilityButton
   parent=HorizontalBox
   widget=Cancel_Button
   properties {widgetStyle {
     clickedSlateSound {resourceObject=null}
     disabled {
      bIsDynamicallyLoaded=false
      drawAs=RoundedBox
      imageType=NoImage
      mirroring=NoMirror
      resourceName=None
      resourceObject=null
      tiling=NoTile
      imageSize=<0 0>
      margin=<ltrb 0 0 0 0>
      outlineSettings {
       bUseBrushTransparency=false
       roundingType=FixedRadius
       width=1
       color {colorUseRule=UseColor_ColorTable, specifiedColor=<rgba 0 0 0 0>}
       cornerRadii=<4 4 4 4>
      }
      tintColor {colorUseRule=UseColor_Specified, specifiedColor=<rgba 0.0395 0.0395 0.0395 1>}
      uVRegion {bIsValid=false, max=<0 0>, min=<0 0>}
     }
     disabledForeground {colorUseRule=UseColor_ColorTable, specifiedColor=<rgba 0 0 0 0>}
     hovered {
      bIsDynamicallyLoaded=false
      drawAs=RoundedBox
      imageType=NoImage
      mirroring=NoMirror
      resourceName=None
      resourceObject=null
      tiling=NoTile
      imageSize=<32 32>
      margin=<ltrb 0 0 0 0>
      outlineSettings {
       bUseBrushTransparency=false
       roundingType=FixedRadius
       width=1
       color {colorUseRule=UseColor_ColorTable, specifiedColor=<rgba 0 0 0 0>}
       cornerRadii=<4 4 4 4>
      }
      tintColor {colorUseRule=UseColor_Specified, specifiedColor=<rgba 0.0953 0.0953 0.0953 1>}
      uVRegion {bIsValid=false, max=<0 0>, min=<0 0>}
     }
     hoveredForeground {colorUseRule=UseColor_ColorTable, specifiedColor=<rgba 0 0 0 0>}
     hoveredSlateSound {resourceObject=null}
     normal {
      bIsDynamicallyLoaded=false
      drawAs=RoundedBox
      imageType=NoImage
      mirroring=NoMirror
      resourceName=None
      resourceObject=null
      tiling=NoTile
      imageSize=<32 32>
      margin=<ltrb 0 0 0 0>
      outlineSettings {
       bUseBrushTransparency=false
       roundingType=FixedRadius
       width=1
       color {colorUseRule=UseColor_ColorTable, specifiedColor=<rgba 0 0 0 0>}
       cornerRadii=<4 4 4 4>
      }
      tintColor {colorUseRule=UseColor_Specified, specifiedColor=<rgba 0.2705 0 0.0107 1>}
      uVRegion {bIsValid=false, max=<0 0>, min=<0 0>}
     }
     normalForeground {colorUseRule=UseColor_ColorTable, specifiedColor=<rgba 0 0 0 0>}
     normalPadding=<ltrb 12 1.5 12 1.5>
     pressed {
      bIsDynamicallyLoaded=false
      drawAs=RoundedBox
      imageType=NoImage
      mirroring=NoMirror
      resourceName=None
      resourceObject=null
      tiling=NoTile
      imageSize=<32 32>
      margin=<ltrb 0 0 0 0>
      outlineSettings {
       bUseBrushTransparency=false
       roundingType=FixedRadius
       width=1
       color {colorUseRule=UseColor_ColorTable, specifiedColor=<rgba 0 0 0 0>}
       cornerRadii=<4 4 4 4>
      }
      tintColor {colorUseRule=UseColor_Specified, specifiedColor=<rgba 0.0284 0.0284 0.0284 1>}
      uVRegion {bIsValid=false, max=<0 0>, min=<0 0>}
     }
     pressedForeground {colorUseRule=UseColor_ColorTable, specifiedColor=<rgba 0 0 0 0>}
     pressedPadding=<ltrb 12 2.5 12 0.5>
     pressedSlateSound {resourceObject=null}
    }}
   slot {class=HorizontalBoxSlot, properties {size {sizeRule=Fill, value=1}}}
  }
  TextBlock_4 {
   widgetClassPath=&TextBlock
   parent=Cancel_Button
   widget=TextBlock_4
   properties {
    text=Cancel
    font {
     bForceMonospaced=false
     bMaterialIsStencil=false
     fontMaterial=null
     fontObject=&Roboto
     letterSpacing=0
     monospacedWidth=1
     size=15
     skewAmount=0
     typefaceFontName=Bold
     outlineSettings {
      bApplyOutlineToDropShadows=false
      bMiteredCorners=false
      bSeparateFillAlpha=false
      outlineMaterial=null
      outlineSize=0
      outlineColor=<rgba 0 0 0 1>
     }
    }
   }
   slot {class=ButtonSlot}
  }
  OutputLog_EditorUtilityExpandableArea {
   widgetClassPath=&EditorUtilityExpandableArea
   bIsVariable=true
   parent=VerticalBox_18
   widget=OutputLog_EditorUtilityExpandableArea
   properties {
    bIsExpanded=true
    borderBrush {
     bIsDynamicallyLoaded=false
     drawAs=RoundedBox
     imageType=NoImage
     mirroring=NoMirror
     resourceName=None
     resourceObject=null
     tiling=NoTile
     imageSize=<0 0>
     margin=<ltrb 0 0 0 0>
     outlineSettings {
      bUseBrushTransparency=false
      roundingType=FixedRadius
      width=0
      color {colorUseRule=UseColor_Specified, specifiedColor=<rgba 0 0 0 0>}
      cornerRadii=<4 4 4 4>
     }
     tintColor {colorUseRule=UseColor_Specified, specifiedColor=<rgba 0 0 0 1>}
     uVRegion {bIsValid=false, max=<0 0>, min=<0 0>}
    }
   }
   slot {class=VerticalBoxSlot}
  }
  TextBlock_3 {
   widgetClassPath=&TextBlock
   namedSlotHost=OutputLog_EditorUtilityExpandableArea
   widget=TextBlock_3
   properties {
    text=Output Log
    font {
     bForceMonospaced=false
     bMaterialIsStencil=false
     fontMaterial=null
     fontObject=&Roboto
     letterSpacing=0
     monospacedWidth=1
     size=15
     skewAmount=0
     typefaceFontName=Italic
     outlineSettings {
      bApplyOutlineToDropShadows=false
      bMiteredCorners=false
      bSeparateFillAlpha=false
      outlineMaterial=null
      outlineSize=0
      outlineColor=<rgba 0 0 0 1>
     }
    }
   }
  }
  VerticalBox {widgetClassPath=&VerticalBox, namedSlotHost=OutputLog_EditorUtilityExpandableArea, widget=VerticalBox}
  Output_TextBlock {
   widgetClassPath=/Script/UMG.RichTextBlock
   bIsVariable=true
   parent=VerticalBox
   widget=Output_TextBlock
   properties {textStyleSet=&/DT_IkFKMatching_TextBlock}
   slot {class=VerticalBoxSlot}
  }
  SetACFromPicker_Button {
   widgetClassPath=&EditorUtilityButton
   parent=VerticalBox
   widget=SetACFromPicker_Button
   properties {visibility=Hidden}
   slot {class=VerticalBoxSlot}
  }
 ]
}
