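# Editor utility widget that cleans up arm and/or leg twist on biped Control Rigs in Sequencer through an IK/FK round-trip bake.
# Run configures a transient EUW_IkFkMatching widget from the biped matching data asset and bakes chosen limbs to IK across the playback range.
# Its finish dispatcher is bound, apparently to the cleanup function, which deletes float keys on limb-named channels, then rebakes with Bake To IK off.
# A Done flag stops further passes.
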EditorUtilityWidgetBlueprint /AnimInEdTools/IkFkMatchingTool/EUW_Limb_Twist_Cleanup
graphs [
 Delete FK Keys and Call Rebake {nodes [
   Delete FK Keys and Call Rebake() {is=title K2Node_FunctionEntry}
   > n1 = Branch(Condition=Get Done() -> (Done:bool))
   n2 = Sequence(execute=@n1.else)
   > n3 = GetControlRigs!(execute, LevelSequence:Level Sequence Object Reference=Get Level Sequence() -> (Level Sequence:Level Sequence Object Reference)) -> (then, ReturnValue:Array of Control Rig Sequencer Binding Proxy Structures)
   > n4 = For Each Loop(Array:Array of Control Rig Sequencer Binding Proxy Structures=@n3) -> (Array Element:Control Rig Sequencer Binding Proxy Structure)
   > n5 = Set Control rig sequencer proxy!(execute, Control rig sequencer proxy:Control Rig Sequencer Binding Proxy Structure=@n4) -> (
    then
    Output_Get_Proxy:Movie Scene Binding Proxy Structure
    Output_Get_ControlRig:Control Rig Object Reference
    Output_Get_Track:Movie Scene Control Rig Parameter Track Object Reference
   )
   > Set Local Proxy(Local Proxy:Movie Scene Binding Proxy Structure=@n5) -> (Output_Get:Movie Scene Binding Proxy Structure)
   > Branch(Condition="Equal (Object)"(A=@n5.Output_Get_ControlRig, B=Get Active Control Rig() -> (Active Control Rig:Control Rig Object Reference)))
   > n6 = LoadAsset_Blocking(Asset=/AnimInEdTools/EpicControlRig/DA_IkFkMatchingData_Biped_v02)
   > n7 = Cast To PDA_IkFkMatchingBakeData!(execute, Object=@n6) -> (then, CastFailed, AsPDA Ik Fk Matching Bake Data:PDA Ik Fk Matching Bake Data Object Reference) {is=title K2Node_DynamicCast}
   > Set Local IKFK Data Array(Local IKFK Data Array:Array of Ik Fk Matching Data Structures=Get Ik Fk Matching Data!(self:PDA Ik Fk Matching Bake Data Object Reference=@n7) -> (Ik Fk Matching Data:Array of Ik Fk Matching Data Structures)) -> (Output_Get:Array of Ik Fk Matching Data Structures)
   > n8 = For Each Loop(Array:Array of Ik Fk Matching Data Structures=Get Local IKFK Data Array() -> (Local IKFK Data Array:Array of Ik Fk Matching Data Structures)) -> (Array Element:Ik Fk Matching Data Structure)
   > Set Local IKFK Data Struct(Local IKFK Data Struct:Ik Fk Matching Data Structure=@n8) -> (Output_Get:Ik Fk Matching Data Structure)
   n9 = GetSelectedIndex!(self:"Combo Box (String) Object Reference"=Get LimbCombobox() -> (LimbCombobox:Editor Utility Combo Box String Object Reference)) -> (ReturnValue:int=0)
   > n10 = Branch(Condition="Equal (Integer)"(A=@n9))
   > n11 = Branch(Condition="integer < integer"(A=@n8.Array Index, B=2))
   n12 = GetTracks!(
    execute=[@n11.then, @n22.then, @n23.then]
    InBinding:"Movie Scene Binding Proxy Structure (by ref)"=Get Local Proxy() -> (Local Proxy:Movie Scene Binding Proxy Structure)
   ) -> (then, ReturnValue:Array of Movie Scene Track Object References)
   > n13 = For Each Loop(Array:Array of Movie Scene Track Object References=@n12) -> (Array Element:Movie Scene Track Object Reference)
   > n14 = GetSections!(execute, Track:Movie Scene Track Object Reference=@n13) -> (then, ReturnValue:Array of Movie Scene Section Object References)
   > n15 = For Each Loop(Array:Array of Movie Scene Section Object References=@n14) -> (Array Element:Movie Scene Section Object Reference)
   > n16 = GetAllChannels!(execute, Section:Movie Scene Section Object Reference=@n15) -> (then, ReturnValue:Array of Movie Scene Scripting Channel Object References)
   > n17 = For Each Loop(Array:Array of Movie Scene Scripting Channel Object References=@n16) -> (Array Element:Movie Scene Scripting Channel Object Reference)
   > n18 = Cast To MovieSceneScriptingFloatChannel!(execute, Object=@n17) -> (then, CastFailed, AsMovie Scene Scripting Float Channel:Movie Scene Scripting Float Channel Object Reference) {is=title K2Node_DynamicCast}
   > n19 = For Each Loop(Array:Array of Movie Scene Scripting Key Object References="Get Keys (Float)"!(self:Movie Scene Scripting Float Channel Object Reference=@n18) -> (ReturnValue:Array of Movie Scene Scripting Key Object References) {is=title K2Node_CallFunction}) -> (Array Element:Movie Scene Scripting Key Object Reference)
   n20 = Get!() -> (
    Local IKFK Data Struct_Name_19_#1:Name
    Local IKFK Data Struct_IkFkSwitchAttr_20_#2:Name
    Local IKFK Data Struct_MatchIKEvent_21_#3:Name
    Local IKFK Data Struct_MatchFKEvent_22_#4:Name
    Local IKFK Data Struct_IKControls_10_#5:Array of Names
    Local IKFK Data Struct_FKControls_12_#6:Array of Names
   )
   > Branch(Condition=Contains(SearchIn=GetDisplayName(Object=@n18), Substring="To String (Name)"(InName=@n20)))
   > "Remove Key (Float)"!(execute, self:Movie Scene Scripting Float Channel Object Reference=@n18, Key:Movie Scene Scripting Key Object Reference=@n19) -> (then)
   n21 = Branch(execute=@n10.else, Condition="Equal (Integer)"(A=@n9, B=1))
   > n22 = Branch(Condition="integer > integer"(A=@n8.Array Index, B=1))
   n23 = Branch(execute=@n21.else, Condition="Equal (Integer)"(A=@n9, B=2))
   Set Bake To IK(execute=@n2.then_1, Bake To IK:bool=false) -> (Output_Get:bool)
   > IKFK_V2_Match_Event()
   > Set Done(Done:bool=true) -> (Output_Get:bool)
  ]}
 EventGraph {nodes [
   "On Pressed (RunButton)"() {is=title K2Node_ComponentBoundEvent}
   > Set Bake To IK(Bake To IK=true)
   > Set Done(Done=false)
   n1 = Get ProxycomboBox() -> (ProxycomboBox:Editor Utility Combo Box String Object Reference)
   > Set Active Control Rig(Active Control Rig:Control Rig Object Reference=Find(
     TargetMap:Map of Strings to Control Rig Object References=Get control rig asset combo box map() -> (control rig asset combo box map:Map of Strings to Control Rig Object References)
     Key=GetSelectedOption!(self:"Combo Box (String) Object Reference"=@n1) -> (ReturnValue:String)
    ) -> (Value:Control Rig Object Reference)) -> (Output_Get:Control Rig Object Reference)
   > n2 = Get Frames!(execute, self) -> (then, valid frames:bool, start frame:int=0, end frame:int=0) {is=title K2Node_CallFunction}
   > n3 = Set!(
    execute
    Range_bHasStart:bool=true
    Range_bHasEnd:bool=true
    Range_InclusiveStart:int=int - int(A=@n2.start frame)
    Range_ExclusiveEnd:int=@n2.end frame
   ) -> (then, Output_Get:Sequencer Scripting Range Structure)
   DelayUntilNextTick!(execute=[@n9.then, @n3.then], WorldContextObject:Object Reference) -> (then)
   > Set Current Local Time!(
    execute
    PlaybackParams_Frame_FrameNumber_Value:int=Get!() -> (Range_bHasStart:bool=~, Range_bHasEnd:bool=~, Range_InclusiveStart:int, Range_ExclusiveEnd:int).Range_InclusiveStart
    PlaybackParams_Frame_SubFrame:float=0
    PlaybackParams_Time:float=0
    PlaybackParams_MarkedFrame:String
    PlaybackParams_Timecode:Timecode Structure=(Hours=0,Minutes=0,Seconds=0,Frames=0,Subframe=0,bDropFrameFormat=False)
    PlaybackParams_PositionType:EMovieScenePositionType Enum=Frame
    PlaybackParams_UpdateMethod:EUpdatePositionMethod Enum=Play
    PlaybackParams_bHasJumped:bool=False
    TimeUnit:EMovieSceneTimeUnit Enum=DisplayRate
   ) -> (then) {is=title K2Node_CallFunction}
   > IKFKMatch_BipedV2!(execute, self, rig:Control Rig Object Reference=Get Active Control Rig()) -> (then)
   "On Pressed (RefreshToolBtn)"() {is=title K2Node_ComponentBoundEvent}
   > n4 = ClearOptions(self:"Combo Box (String) Object Reference"=Get LimbCombobox())
   n5 = Sequence(execute=[@n4.then, @n10.then])
   > n6 = GetCurrentLevelSequence!(execute) -> (then, ReturnValue:Level Sequence Object Reference)
   > Set Level Sequence(Level Sequence:Level Sequence Object Reference=@n6) -> (Output_Get:Level Sequence Object Reference)
   > n7 = Branch(Condition=IsLoadingAssets!(self:Asset Registry Interface=GetAssetRegistry()) -> (ReturnValue:bool))
   Populate Control Rig Combo Box!(
    execute=@n7.else
    self
    Local Level Sequence:Level Sequence Object Reference=Get Level Sequence()
    Local Combo Box:Editor Utility Combo Box String Object Reference=Get ProxycomboBox()
   ) -> (then)
   n8 = Get LimbCombobox()
   AddOption!(execute=@n5.then_1, self:"Combo Box (String) Object Reference"=@n8, Option:String=Arms) -> (then)
   > AddOption(self=@n8, Option=Legs)
   > AddOption(self=@n8, Option=Both)
   > SetSelectedIndex!(execute, self:"Combo Box (String) Object Reference"=@n8, Index:int=0) -> (then)
   n9 = IKFK_V2_Match_Event() {is=title K2Node_CustomEvent}
   n10 = Event Construct()
   Event Tick()
   Event PreConstruct()
  ]}
 Get Frames {nodes [
   Get Frames() {is=title K2Node_FunctionEntry}
   > n1 = Sequence()
   > n2 = Execute Python Script!(
    execute
    PythonScript="start = level_sequence.get_playback_start() \r\nend = level_sequence.get_playback_end()"
    level_sequence:Level Sequence Object Reference=Get Level Sequence()
   ) -> (then, ReturnValue, start:int, end:int) {inputs [level_sequence], outputs [start,end]}
   > Set frame start(frame start:int=@n2.start) -> (Output_Get:int=0)
   > Set frame end(frame end:int=@n2.end) -> (Output_Get:int=0)
   > Set local valid frames(local valid frames:bool=true) -> (Output_Get:bool)
   Return Node!(
    execute=@n1.then_1
    valid frames:bool=Get local valid frames() -> (local valid frames:bool)
    start frame:int=Get frame start() -> (frame start:int=0)
    end frame:int=Get frame end() -> (frame end:int=0)
   )
   n3 = Branch(Condition=IsChecked(self=Get PropFullRangeCbox() -> (PropFullRangeCbox:Editor Utility Check Box Object Reference)))
   n4 = Execute Python Script!(
    execute=@n3.else
    PythonScript="from AnimLibrary.core.utils import misc\r\n\r\nvalid_frames = True\r\n\r\nstart, end = misc.get_selected_keys_range()\r\n\r\nif not start:\r\n   valid_frames = False\r\n   print('Start Frame is invalid')\r\n   start = 0\r\n\r\nif not end:\r\n    valid_frames = False\r\n    end = 0\r\n    print('End Frame is invalid')\r\n"
   ) -> (then, ReturnValue, start, end, valid_frames:bool=~) {outputs [start,end,valid_frames]}
   > n5 = Set local valid frames(local valid frames=@n4.valid_frames)
   > Branch(Condition=@n5)
   > Set frame start(frame start=@n4.start) -> (Output_Get=0)
   > Set frame end(frame end=@n4.end) -> (Output_Get=0)
  ]}
 IKFKMatch_BipedV2 {nodes [
   IKFKMatch_BipedV2!() -> (then, rig:Control Rig Object Reference) {is=title K2Node_FunctionEntry}
   > n1 = LoadAsset_Blocking(Asset=&DA_IkFkMatchingData_Biped_v02)
   > n2 = Cast To PDA_IkFkMatchingBakeData(Object=@n1)
   > Set Local IKFK Data Array(Local IKFK Data Array=Get Ik Fk Matching Data(self=@n2))
   > n3 = Create Widget!(
    execute:Exec
    WorldContextObject:Object Reference
    Class:User Widget Class Reference=&/EUW_IkFkMatching.EUW_IkFkMatching_C
    OwningPlayer:Player Controller Object Reference
    Primary Control Rig:Control Rig Object Reference
    Animatable Creature:Animatable Creature Object Reference
   ) -> (then:Exec, ReturnValue:EUW Ik Fk Matching Object Reference)
   > n4 = Set Local EUW IKFK Matcher(Local EUW IKFK Matcher:EUW Ik Fk Matching Object Reference=@n3) -> (Output_Get:EUW Ik Fk Matching Object Reference)
   > Bind Event to IKFK Match Finished E Dispatcher(self:EUW Ik Fk Matching Object Reference=@n4, Delegate=Create Event()) {is=title K2Node_AddDelegate}
   > n5 = Cast To ModularRig!(execute, Object=Get Active Control Rig()) -> (then, CastFailed, AsModular Rig:Modular Rig Object Reference) {is=title K2Node_DynamicCast}
   > n6 = Sequence!(execute) -> (then_0, then_1, then_2:Exec, then_3:Exec)
   > n7 = GetControlRigs(LevelSequence=Get Level Sequence())
   > n8 = For Each Loop(Array:Array of Control Rig Sequencer Binding Proxy Structures=@n7) -> (Array Element:Control Rig Sequencer Binding Proxy Structure)
   > n9 = Set control rig sequencer proxy!(execute, control rig sequencer proxy:Control Rig Sequencer Binding Proxy Structure=@n8) -> (
    then
    Output_Get_Proxy:Movie Scene Binding Proxy Structure
    Output_Get_ControlRig:Control Rig Object Reference
    Output_Get_Track:Movie Scene Control Rig Parameter Track Object Reference
   )
   > Set control rig sequencer proxy!(execute, control rig sequencer proxy=@n8) -> (then, Output_Get:Control Rig Sequencer Binding Proxy Structure)
   > Branch(Condition="Equal (Object)"(A=Get Active Control Rig(), B=@n9.Output_Get_ControlRig))
   > Set Modular Rig!(
    execute
    Modular Rig:Modular Rig Object Reference=@n5
    self:EUW Ik Fk Matching Object Reference=Get Local EUW IKFK Matcher() -> (Local EUW IKFK Matcher:EUW Ik Fk Matching Object Reference)
   ) -> (then, Output_Get:Modular Rig Object Reference)
   > Set!(execute, Frame Rate_Numerator:int=30, Frame Rate_Denominator:int=1, self:EUW Ik Fk Matching Object Reference=Get Local EUW IKFK Matcher()) -> (then, Output_Get:Frame Rate Structure)
   > Set Rig Hierarchy!(
    execute
    Rig Hierarchy:Rig Hierarchy Object Reference=GetHierarchy!(self:Control Rig Object Reference=Get Active Control Rig()) -> (ReturnValue:Rig Hierarchy Object Reference)
    self:EUW Ik Fk Matching Object Reference=Get Local EUW IKFK Matcher()
   ) -> (then, Output_Get:Rig Hierarchy Object Reference)
   > SetIsChecked!(
    execute
    self:Check Box Object Reference=Get CurrentFrameCheckBox!(self:EUW Ik Fk Matching Object Reference=Get Local EUW IKFK Matcher()) -> (CurrentFrameCheckBox:Editor Utility Check Box Object Reference)
    InIsChecked:bool=false
   ) -> (then)
   > Set Playback Range!(
    execute
    Playback Range:Sequencer Scripting Range Structure=Get Range() -> (Range:Sequencer Scripting Range Structure)
    self:EUW Ik Fk Matching Object Reference=Get Local EUW IKFK Matcher()
   ) -> (then, Output_Get:Sequencer Scripting Range Structure)
   > Set Current Matching Frame!(
    execute
    Current Matching Frame:int=Get() -> (Range_bHasStart=~, Range_bHasEnd=~).Range_InclusiveStart
    self:EUW Ik Fk Matching Object Reference=Get Local EUW IKFK Matcher()
   ) -> (then, Output_Get:int=0)
   > SetIsChecked(
    self=Get CurrentKeysOnlyCheckBox!(self:EUW Ik Fk Matching Object Reference=Get Local EUW IKFK Matcher()) -> (CurrentKeysOnlyCheckBox:Editor Utility Check Box Object Reference)
    InIsChecked=false
   )
   > Set IsModular!(execute, IsModular:bool=true, self:EUW Ik Fk Matching Object Reference=Get Local EUW IKFK Matcher()) -> (then, Output_Get:bool)
   > n10 = For Each Loop(Array:Array of Ik Fk Matching Data Structures=Get Local IKFK Data Array()) -> (Array Element:Ik Fk Matching Data Structure)
   > Set Local IKFK Data Struct(Local IKFK Data Struct=@n10)
   n11 = GetSelectedIndex(self=Get LimbCombobox()) -> (ReturnValue=0)
   > n12 = Branch(Condition="Equal (Integer)"(A=@n11))
   > n13 = Branch(Condition="integer < integer"(A=@n10.Array Index, B=2))
   n14 = GetSections(
    execute=[@n13.then, @n23.then, @n24.then]
    Track=Break Control Rig Sequencer Binding Proxy!(ControlRigSequencerBindingProxy:"Control Rig Sequencer Binding Proxy Structure (by ref)"=Get control rig sequencer proxy() -> (control rig sequencer proxy:Control Rig Sequencer Binding Proxy Structure)) -> (
     Proxy:Movie Scene Binding Proxy Structure
     ControlRig:Control Rig Object Reference
     Track:Movie Scene Control Rig Parameter Track Object Reference
    ) {is=title K2Node_BreakStruct}.Track
   )
   > n15 = For Each Loop(Array:Array of Movie Scene Section Object References=@n14) -> (Array Element:Movie Scene Section Object Reference)
   > n16 = GetAllChannels(Section=@n15)
   > n17 = For Each Loop(Array:Array of Movie Scene Scripting Channel Object References=@n16) -> (Array Element:Movie Scene Scripting Channel Object Reference)
   n18 = Break Ik Fk Matching Data!(IkFkMatchingData:"Ik Fk Matching Data Structure (by ref)"=Get Local IKFK Data Struct() -> (Local IKFK Data Struct:Ik Fk Matching Data Structure)) -> (
    Name_19_#1:Name
    IkFkSwitchAttr_20_#2:Name
    MatchIKEvent_21_#3:Name
    MatchFKEvent_22_#4:Name
    IKControls_10_#5:Array of Names
    FKControls_12_#6:Array of Names
   ) {is=title K2Node_BreakStruct}
   n19 = "To String (Name)"(InName=@n18)
   n20 = "To String (Name)"(InName=@n18.IkFkSwitchAttr_20_#2)
   > Branch(Condition=Contains(SearchIn=GetDisplayName(Object=@n17), Substring=Append!(A=@n19, B=/, C:String=@n20) -> (ReturnValue)))
   > n21 = Cast To MovieSceneScriptingBoolChannel!(execute, Object=@n17) -> (then, CastFailed, AsMovie Scene Scripting Bool Channel:Movie Scene Scripting Bool Channel Object Reference) {is=title K2Node_DynamicCast}
   > Add!(
    execute
    TargetMap:Map of Movie Scene Scripting Bool Channel Object References to Booleans=Get IK FK Switch Channels!(self:EUW Ik Fk Matching Object Reference=Get Local EUW IKFK Matcher()) -> (IK FK Switch Channels:Map of Movie Scene Scripting Bool Channel Object References to Booleans)
    Key:"Movie Scene Scripting Bool Channel Object Reference (by ref)"=@n21
    Value:"Boolean (by ref)"=Get Bake To IK() -> (Bake To IK:bool)
   ) -> (then) {is=title K2Node_CallFunction}
   > Add(
    TargetArray:Array of Ik Fk Matching Data Structures=Get Ik Fk Matching Data(self:EUW Ik Fk Matching Object Reference=Get Local EUW IKFK Matcher())
    NewItem:"Ik Fk Matching Data Structure (by ref)"=Get Local IKFK Data Struct()
   ) {is=title K2Node_CallArrayFunction}
   > Add(
    TargetArray:Array of Strings=Get Module Paths!(self:EUW Ik Fk Matching Object Reference=Get Local EUW IKFK Matcher()) -> (Module Paths:Array of Strings)
    NewItem:"String (by ref)"=@n19
   )
   n22 = Branch(execute=@n12.else, Condition="Equal (Integer)"(A=@n11, B=1))
   > n23 = Branch(Condition="integer > integer"(A=@n10.Array Index, B=1))
   n24 = Branch(execute=@n22.else, Condition="Equal (Integer)"(A=@n11, B=2))
   Run Match(execute=@n6.then_1, self:EUW Ik Fk Matching Object Reference=Get Local EUW IKFK Matcher())
   Return Node!(execute=@n6.then_2)
  ]}
 Load From Save {nodes [Load From Save() {is=title K2Node_FunctionEntry}]}
 Populate Control Rig Combo Box {nodes [
   Populate Control Rig Combo Box!() -> (then, Local Level Sequence:Level Sequence Object Reference, Local Combo Box:Editor Utility Combo Box String Object Reference) {is=title K2Node_FunctionEntry}
   > n1 = Sequence()
   n2 = GetControlRigs(execute=@n1.then_1, LevelSequence=Get Local Level Sequence() -> (Local Level Sequence:Level Sequence Object Reference))
   > n3 = For Each Loop!(Exec, Array:Array of Control Rig Sequencer Binding Proxy Structures=@n2) -> (
    LoopBody
    Array Element_Proxy:Movie Scene Binding Proxy Structure
    Array Element_ControlRig:Control Rig Object Reference
    Array Element_Track:Movie Scene Control Rig Parameter Track Object Reference
    Array Index
    Completed
   )
   n4 = Get Class(Object=@n3.Array Element_ControlRig) -> (ReturnValue:Control Rig Class Reference)
   > Branch(Condition=OR Boolean(A="Equal (Class)"(A=@n4, B=/EpicControlRig/Templates/MR_Biped_v02.MR_Biped_v02_C), B="Equal (Class)"(A=@n4)))
   > n5 = GetDisplayName!(execute:Exec, InBinding:"Movie Scene Binding Proxy Structure (by ref)"=@n3) -> (then:Exec, ReturnValue:Text)
   n6 = "To String (Text)"(InText=@n5)
   > AddOption(self=Get Local Combo Box() -> (Local Combo Box:Editor Utility Combo Box String Object Reference), Option=@n6)
   > Add(
    TargetMap:Map of Strings to Control Rig Object References=Get control rig asset combo box map()
    Key:"String (by ref)"=@n6
    Value:"Control Rig Object Reference (by ref)"=@n3.Array Element_ControlRig
   ) {is=title K2Node_CallFunction}
   SetSelectedIndex(execute=@n1.then_2, self=Get Local Combo Box(), Index=0)
   Return Node(execute=@n1.then_3)
   ClearOptions(execute=@n1.then_0, self=Get Local Combo Box())
   Get Selected Parent Rig() -> (Selected Parent Rig:Control Rig Object Reference)
  ]}
 Set Control Rig Controls for Combo Box {nodes [
   Set Control Rig Controls for Combo Box!() -> (
    then
    Selected Control Rig:Control Rig Object Reference
    Combo Box:"Searchable Combo Box (String) Object Reference"
    Socket attachment:ECheckBoxState Enum=Unchecked
   ) {is=title K2Node_FunctionEntry}
   > Return Node!(execute, Attach Point Array:Array of Strings)
  ]}
]
variables [
 Active Control Rig {type=ControlRig, default=null}
 Bake To IK {type=boolean, default=true}
 control rig asset combo box map {type="map<ControlRig>", default {}}
 Done {type=boolean}
 frame end {type=integer}
 frame start {type=integer}
 IKFK Data Struct {type=IkFkMatchingData, default {name=None, ik Fk Switch Attr=None, match FK Event=None, match IK Event=None, fK Controls [], iK Controls []}}
 Level Sequence {type=LevelSequence, default=null}
 local valid frames {type=boolean}
 Range {
  type=SequencerScriptingRange
  default {bHasEnd=false, bHasStart=false, exclusiveEnd=0, inclusiveStart=0, internalRate {denominator=1, numerator=60000}}
 }
 Search Words {type="array<string>", default []}
]
widgets {
 info {rootWidgetClass=&CanvasPanel, widgetCount=12}
 widgets [
  CanvasPanel_10 {widgetClassPath=&CanvasPanel, widget=CanvasPanel_10}
  VerticalBox_19 {
   widgetClassPath=&VerticalBox
   parent=CanvasPanel_10
   widget=VerticalBox_19
   slot {class=CanvasPanelSlot, properties {layoutData {alignment=<0 0>, anchors {maximum=<0 0>, minimum=<0 0>}, offsets=<ltrb 0 0 589.2165 30>}}}
  }
  HorizontalBox_0 {widgetClassPath=&HorizontalBox, parent=VerticalBox_19, widget=HorizontalBox_0, slot {class=VerticalBoxSlot}}
  combboxtext {
   widgetClassPath=&TextBlock
   parent=HorizontalBox_0
   widget=combboxtext
   properties {
    text="Asset: "
    font {
     bForceMonospaced=false
     bMaterialIsStencil=false
     fontMaterial=null
     fontObject=&Roboto
     letterSpacing=0
     monospacedWidth=1
     size=13.5
     skewAmount=0
     typefaceFontName=Bold
     outlineSettings {
      bApplyOutlineToDropShadows=false
      bMiteredCorners=false
      bSeparateFillAlpha=false
      outlineMaterial=null
      outlineSize=0
      outlineColor=<rgba 0 0 0 1>
     }
    }
   }
   slot {class=HorizontalBoxSlot, properties {verticalAlignment=VAlign_Center}}
  }
  ProxycomboBox {
   widgetClassPath=/Script/Blutility.EditorUtilityComboBoxString
   bIsVariable=true
   parent=HorizontalBox_0
   widget=ProxycomboBox
   properties {foregroundColor {colorUseRule=UseColor_Specified, specifiedColor=<rgba 0.5271 0.5271 0.5271 1>}}
   slot {class=HorizontalBoxSlot}
  }
  limbcomboboxtext {
   widgetClassPath=&TextBlock
   parent=HorizontalBox_0
   widget=limbcomboboxtext
   properties {
    text="Limb: "
    font {
     bForceMonospaced=false
     bMaterialIsStencil=false
     fontMaterial=null
     fontObject=&Roboto
     letterSpacing=0
     monospacedWidth=1
     size=13.5
     skewAmount=0
     typefaceFontName=Bold
     outlineSettings {
      bApplyOutlineToDropShadows=false
      bMiteredCorners=false
      bSeparateFillAlpha=false
      outlineMaterial=null
      outlineSize=0
      outlineColor=<rgba 0 0 0 1>
     }
    }
   }
   slot {class=HorizontalBoxSlot, properties {verticalAlignment=VAlign_Center, padding=<ltrb 20 0 0 0>}}
  }
  LimbCombobox {
   widgetClassPath=&EditorUtilityComboBoxString
   bIsVariable=true
   parent=HorizontalBox_0
   widget=LimbCombobox
   properties {foregroundColor {colorUseRule=UseColor_Specified, specifiedColor=<rgba 0.5271 0.5271 0.5271 1>}}
   slot {class=HorizontalBoxSlot}
  }
  HorizontalBox_1 {
   widgetClassPath=&HorizontalBox
   parent=VerticalBox_19
   widget=HorizontalBox_1
   slot {class=VerticalBoxSlot, properties {horizontalAlignment=HAlign_Left}}
  }
  RunButton {
   widgetClassPath=&EditorUtilityButton
   parent=HorizontalBox_1
   widget=RunButton
   slot {class=HorizontalBoxSlot, properties {horizontalAlignment=HAlign_Left, size {sizeRule=Fill, value=1}}}
  }
  RunText {
   widgetClassPath=&TextBlock
   parent=RunButton
   widget=RunText
   properties {
    text=Run Process
    font {
     bForceMonospaced=false
     bMaterialIsStencil=false
     fontMaterial=null
     fontObject=&Roboto
     letterSpacing=0
     monospacedWidth=1
     size=15
     skewAmount=0
     typefaceFontName=Bold
     outlineSettings {
      bApplyOutlineToDropShadows=false
      bMiteredCorners=false
      bSeparateFillAlpha=false
      outlineMaterial=null
      outlineSize=0
      outlineColor=<rgba 0 0 0 1>
     }
    }
   }
   slot {class=ButtonSlot}
  }
  RefreshToolBtn {
   widgetClassPath=&EditorUtilityButton
   parent=HorizontalBox_1
   widget=RefreshToolBtn
   properties {toolTipText=Refresh UI}
   slot {class=HorizontalBoxSlot, properties {horizontalAlignment=HAlign_Left, padding=<ltrb 10 0 0 0>}}
  }
  Image_54 {
   widgetClassPath=&Image
   bIsVariable=true
   parent=RefreshToolBtn
   widget=Image_54
   properties {
    visibility=HitTestInvisible
    brush {
     bIsDynamicallyLoaded=false
     drawAs=Image
     imageType=FullColor
     mirroring=NoMirror
     resourceName=None
     resourceObject=null
     tiling=NoTile
     imageSize=<24 24>
     margin=<ltrb 0 0 0 0>
     outlineSettings {
      bUseBrushTransparency=false
      roundingType=HalfHeightRadius
      width=0
      color {colorUseRule=UseColor_Specified, specifiedColor=<rgba 0 0 0 0>}
      cornerRadii=<0 0 0 0>
     }
     tintColor {colorUseRule=UseColor_Specified, specifiedColor=<rgba 1 1 1 1>}
     uVRegion {bIsValid=false, max=<0 0>, min=<0 0>}
    }
   }
   slot {class=ButtonSlot, properties {horizontalAlignment=HAlign_Left}}
  }
 ]
}
